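# Physics asset for the female medium Jellyfish body cosmetic, setting up secondary motion for its two-tier skirt and clam necklace.
# Selected skirt panels have capsules with narrow angular limits on small Base boxes, which join Const_Base boxes through free-rotation constraints.
# Only right front and side Const_Base boxes constrain to thigh_sub_r; left and back ones have no parent constraint.
# The necklace chains from spine_05 through two links to clamTop, twist-locked at the root.
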PhysicsAsset /BRCosmetics/Characters/Player/Female/Medium/Bodies/F_MED_Jellyfish/Meshes/F_MED_Jellyfish_Physics
bodies {
 dyn_clamTop [{shapeType=Capsule, length=3.25, radius=1.75, center=<1.7439 -1.5494 0>}]
 dyn_necklace_01 [{shapeType=Capsule, length=6.75, radius=1, center=<3.1122 -0.0001 0>}]
 dyn_necklace_02 [{shapeType=Capsule, length=3.25, radius=2, center=<2.6272 1.5381 0.3703>, rotation=<rot 0 0 5.0005>}]
 dyn_skirtHi_backA_Base_l [{shapeType=Box, extentX=1.01, extentY=1.01, extentZ=1.01, rotation=<rot 90 0 0>}]
 dyn_skirtHi_backA_Base_r [{shapeType=Box, extentX=1.01, extentY=1.01, extentZ=1.01, rotation=<rot 0 -90 0>}]
 dyn_skirtHi_backA_Const_Base_l [{shapeType=Box, extentX=1.01, extentY=1.01, extentZ=1.01, rotation=<rot 90 0 0>}]
 dyn_skirtHi_backA_Const_Base_r [{shapeType=Box, extentX=1.01, extentY=1.01, extentZ=1.01, rotation=<rot 0 -90 0>}]
 dyn_skirtHi_backA_r [{shapeType=Capsule, length=8, radius=3, center=<3.801 -0.0001 0>, rotation=<rot 90 0 0>}]
 dyn_skirtHi_backB_Base_l [{shapeType=Box, extentX=1.01, extentY=1.01, extentZ=1.01, rotation=<rot 90 0 0>}]
 dyn_skirtHi_backB_Base_r [{shapeType=Box, extentX=1.01, extentY=1.01, extentZ=1.01, rotation=<rot 0 -90 0>}]
 dyn_skirtHi_backB_Const_Base_l [{shapeType=Box, extentX=1.01, extentY=1.01, extentZ=1.01, rotation=<rot 90 0 0>}]
 dyn_skirtHi_backB_Const_Base_r [{shapeType=Box, extentX=1.01, extentY=1.01, extentZ=1.01, rotation=<rot 0 -90 0>}]
 dyn_skirtHi_backB_l [{shapeType=Capsule, length=6.75, radius=3, center=<3.801 -0.0001 0>, rotation=<rot 90 0 0>}]
 dyn_skirtHi_backB_r [{shapeType=Capsule, length=9.5, radius=3, center=<3.801 -0.0001 0>, rotation=<rot 90 0 0>}]
 dyn_skirtHi_frontA_Base_l [{shapeType=Box, extentX=1.01, extentY=1.01, extentZ=1.01, rotation=<rot 0 -90 0>}]
 dyn_skirtHi_frontA_Base_r [{shapeType=Box, extentX=1.01, extentY=1.01, extentZ=1.01}]
 dyn_skirtHi_frontA_Const_Base_l [{shapeType=Box, extentX=1.01, extentY=1.01, extentZ=1.01, rotation=<rot 0 -90 0>}]
 dyn_skirtHi_frontA_Const_Base_r [{shapeType=Box, extentX=1.01, extentY=1.01, extentZ=1.01}]
 dyn_skirtHi_frontA_r [{shapeType=Capsule, length=9, radius=4, center=<3.801 -0.0001 0>, rotation=<rot 90 0 -35.0023>}]
 dyn_skirtHi_frontB_Base_l [{shapeType=Box, extentX=1.01, extentY=1.01, extentZ=1.01, rotation=<rot 0 -90 0>}]
 dyn_skirtHi_frontB_Base_r [{shapeType=Box, extentX=1.01, extentY=1.01, extentZ=1.01}]
 dyn_skirtHi_frontB_Const_Base_l [{shapeType=Box, extentX=1.01, extentY=1.01, extentZ=1.01, rotation=<rot 0 -90 0>}]
 dyn_skirtHi_frontB_Const_Base_r [{shapeType=Box, extentX=1.01, extentY=1.01, extentZ=1.01}]
 dyn_skirtHi_frontB_l [{shapeType=Capsule, length=5.5, radius=4.25, center=<3.801 -0.0001 0>, rotation=<rot 80 0 0>}]
 dyn_skirtHi_sideA_Base_l [{shapeType=Box, extentX=1.01, extentY=1.01, extentZ=1.01, rotation=<rot 0 -90 0>}]
 dyn_skirtHi_sideA_Base_r [{shapeType=Box, extentX=1.01, extentY=1.01, extentZ=1.01}]
 dyn_skirtHi_sideA_Const_Base_l [{shapeType=Box, extentX=1.01, extentY=1.01, extentZ=1.01, rotation=<rot 0 -90 0>}]
 dyn_skirtHi_sideA_Const_Base_r [{shapeType=Box, extentX=1.01, extentY=1.01, extentZ=1.01}]
 dyn_skirtHi_sideA_r [{shapeType=Capsule, length=10.75, radius=3, center=<3.801 0.6666 0>, rotation=<rot 90 0 -10.0012>}]
 dyn_skirtHi_sideB_Base_l [{shapeType=Box, extentX=1.01, extentY=1.01, extentZ=1.01, rotation=<rot 0 -90 0>}]
 dyn_skirtHi_sideB_Base_r [{shapeType=Box, extentX=1.01, extentY=1.01, extentZ=1.01}]
 dyn_skirtHi_sideB_Const_Base_l [{shapeType=Box, extentX=1.01, extentY=1.01, extentZ=1.01, rotation=<rot 0 -90 0>}]
 dyn_skirtHi_sideB_Const_Base_r [{shapeType=Box, extentX=1.01, extentY=1.01, extentZ=1.01}]
 dyn_skirtHi_sideB_l [{shapeType=Capsule, length=5.75, radius=4, center=<3.801 -0.0002 -2.2437>, rotation=<rot 70.0002 0 0>}]
 dyn_skirtHi_sideC_Base_r [{shapeType=Box, extentX=1.01, extentY=1.01, extentZ=1.01}]
 dyn_skirtHi_sideC_Const_Base_r [{shapeType=Box, extentX=1.01, extentY=1.01, extentZ=1.01}]
 dyn_skirtHi_sideC_r [{shapeType=Capsule, length=11, radius=3.25, center=<3.501 1.1192 0>, rotation=<rot 90 0 -15.0007>}]
 dyn_skirtLow_backA_Base_l [{shapeType=Box, extentX=1.01, extentY=1.01, extentZ=1.01, rotation=<rot 90 0 0>}]
 dyn_skirtLow_backA_Base_r [{shapeType=Box, extentX=1.01, extentY=1.01, extentZ=1.01, rotation=<rot 0 -90 0>}]
 dyn_skirtLow_backA_Const_Base_l [{shapeType=Box, extentX=1.01, extentY=1.01, extentZ=1.01, rotation=<rot 90 0 0>}]
 dyn_skirtLow_backA_Const_Base_r [{shapeType=Box, extentX=1.01, extentY=1.01, extentZ=1.01, rotation=<rot 0 -90 0>}]
 dyn_skirtLow_backB_Base_r [{shapeType=Box, extentX=1.01, extentY=1.01, extentZ=1.01, rotation=<rot 0 -90 0>}]
 dyn_skirtLow_backB_Const_Base_r [{shapeType=Box, extentX=1.01, extentY=1.01, extentZ=1.01, rotation=<rot 0 -90 0>}]
 dyn_skirtLow_backB_r [{shapeType=Capsule, length=9.5, radius=3, center=<3.801 -0.0001 0>, rotation=<rot 90 0 0>}]
 dyn_skirtLow_frontA_Base_l [{shapeType=Box, extentX=1.01, extentY=1.01, extentZ=1.01, rotation=<rot 0 -90 0>}]
 dyn_skirtLow_frontA_Base_r [{shapeType=Box, extentX=1.01, extentY=1.01, extentZ=1.01}]
 dyn_skirtLow_frontA_Const_Base_l [{shapeType=Box, extentX=1.01, extentY=1.01, extentZ=1.01, rotation=<rot 0 -90 0>}]
 dyn_skirtLow_frontA_Const_Base_r [{shapeType=Box, extentX=1.01, extentY=1.01, extentZ=1.01}]
 dyn_skirtLow_frontA_l [{shapeType=Capsule, length=9.25, radius=4.25, center=<3.801 -0.0001 0>, rotation=<rot 90 0 -15.0005>}]
 dyn_skirtLow_frontB_Base_l [{shapeType=Box, extentX=1.01, extentY=1.01, extentZ=1.01, rotation=<rot 0 -90 0>}]
 dyn_skirtLow_frontB_Base_r [{shapeType=Box, extentX=1.01, extentY=1.01, extentZ=1.01}]
 dyn_skirtLow_frontB_Const_Base_l [{shapeType=Box, extentX=1.01, extentY=1.01, extentZ=1.01, rotation=<rot 0 -90 0>}]
 dyn_skirtLow_frontB_Const_Base_r [{shapeType=Box, extentX=1.01, extentY=1.01, extentZ=1.01}]
 dyn_skirtLow_frontB_r [{shapeType=Capsule, length=6.75, radius=3, center=<3.801 -0.0001 0>, rotation=<rot 90 0 0>}]
 dyn_skirtLow_sideA_Base_l [{shapeType=Box, extentX=1.01, extentY=1.01, extentZ=1.01, rotation=<rot 0 -90 0>}]
 dyn_skirtLow_sideA_Base_r [{shapeType=Box, extentX=1.01, extentY=1.01, extentZ=1.01}]
 dyn_skirtLow_sideA_Const_Base_l [{shapeType=Box, extentX=1.01, extentY=1.01, extentZ=1.01, rotation=<rot 0 -90 0>}]
 dyn_skirtLow_sideA_Const_Base_r [{shapeType=Box, extentX=1.01, extentY=1.01, extentZ=1.01}]
 dyn_skirtLow_sideA_l [{shapeType=Capsule, length=5.75, radius=4, center=<3.801 3.1096 0>, rotation=<rot 90 0 0>}]
 dyn_skirtLow_sideB_Base_l [{shapeType=Box, extentX=1.01, extentY=1.01, extentZ=1.01, rotation=<rot 0 -90 0>}]
 dyn_skirtLow_sideB_Base_r [{shapeType=Box, extentX=1.01, extentY=1.01, extentZ=1.01}]
 dyn_skirtLow_sideB_Const_Base_l [{shapeType=Box, extentX=1.01, extentY=1.01, extentZ=1.01, rotation=<rot 0 -90 0>}]
 dyn_skirtLow_sideB_Const_Base_r [{shapeType=Box, extentX=1.01, extentY=1.01, extentZ=1.01}]
 dyn_skirtLow_sideB_l [{shapeType=Capsule, length=5.75, radius=4, center=<3.8009 1.8541 0>, rotation=<rot 90 0 0>}]
 dyn_skirtLow_sideB_r [{shapeType=Capsule, length=6.75, radius=3, center=<3.579 1.2586 0>, rotation=<rot 90 0 -10.0001>}]
 spine_05 [{shapeType=Capsule, length=15, radius=15, center=<-4.9716 -5.1283 0.0001>, rotation=<rot 0.0006 54.9996 89.9966>}]
 thigh_sub_r [{shapeType=Box, extentX=1.01, extentY=1.01, extentZ=1.01}]
}
constraints [
 {
  bone1Name=dyn_skirtHi_sideC_Const_Base_r
  bone2Name=thigh_sub_r
  swing1LimitDegrees=45
  swing1Motion=Free
  swing2LimitDegrees=45
  swing2Motion=Free
  twistLimitDegrees=45
  twistMotion=Free
 }
 {
  bone1Name=dyn_skirtHi_sideC_Base_r
  bone2Name=dyn_skirtHi_sideC_Const_Base_r
  swing1LimitDegrees=45
  swing1Motion=Free
  swing2LimitDegrees=45
  swing2Motion=Free
  twistLimitDegrees=45
  twistMotion=Free
 }
 {bone1Name=dyn_skirtHi_sideC_r, bone2Name=dyn_skirtHi_sideC_Base_r, swing1LimitDegrees=15, swing2LimitDegrees=15, twistLimitDegrees=10}
 {
  bone1Name=dyn_skirtHi_sideB_Const_Base_r
  bone2Name=thigh_sub_r
  swing1LimitDegrees=45
  swing1Motion=Free
  swing2LimitDegrees=45
  swing2Motion=Free
  twistLimitDegrees=45
  twistMotion=Free
 }
 {
  bone1Name=dyn_skirtHi_sideB_Base_r
  bone2Name=dyn_skirtHi_sideB_Const_Base_r
  swing1LimitDegrees=45
  swing1Motion=Free
  swing2LimitDegrees=45
  swing2Motion=Free
  twistLimitDegrees=45
  twistMotion=Free
 }
 {
  bone1Name=dyn_skirtHi_sideA_Const_Base_r
  bone2Name=thigh_sub_r
  swing1LimitDegrees=45
  swing1Motion=Free
  swing2LimitDegrees=45
  swing2Motion=Free
  twistLimitDegrees=45
  twistMotion=Free
 }
 {
  bone1Name=dyn_skirtHi_sideA_Base_r
  bone2Name=dyn_skirtHi_sideA_Const_Base_r
  swing1LimitDegrees=45
  swing1Motion=Free
  swing2LimitDegrees=45
  swing2Motion=Free
  twistLimitDegrees=45
  twistMotion=Free
 }
 {bone1Name=dyn_skirtHi_sideA_r, bone2Name=dyn_skirtHi_sideA_Base_r, swing1LimitDegrees=15, swing2LimitDegrees=15, twistLimitDegrees=10}
 {
  bone1Name=dyn_skirtHi_frontB_Const_Base_r
  bone2Name=thigh_sub_r
  swing1LimitDegrees=45
  swing1Motion=Free
  swing2LimitDegrees=45
  swing2Motion=Free
  twistLimitDegrees=45
  twistMotion=Free
 }
 {
  bone1Name=dyn_skirtHi_frontB_Base_r
  bone2Name=dyn_skirtHi_frontB_Const_Base_r
  swing1LimitDegrees=45
  swing1Motion=Free
  swing2LimitDegrees=45
  swing2Motion=Free
  twistLimitDegrees=45
  twistMotion=Free
 }
 {
  bone1Name=dyn_skirtHi_frontA_Const_Base_r
  bone2Name=thigh_sub_r
  swing1LimitDegrees=45
  swing1Motion=Free
  swing2LimitDegrees=45
  swing2Motion=Free
  twistLimitDegrees=45
  twistMotion=Free
 }
 {
  bone1Name=dyn_skirtHi_frontA_Base_r
  bone2Name=dyn_skirtHi_frontA_Const_Base_r
  swing1LimitDegrees=45
  swing1Motion=Free
  swing2LimitDegrees=45
  swing2Motion=Free
  twistLimitDegrees=45
  twistMotion=Free
 }
 {bone1Name=dyn_skirtHi_frontA_r, bone2Name=dyn_skirtHi_frontA_Base_r, swing1LimitDegrees=15, swing2LimitDegrees=15, twistLimitDegrees=10}
 {
  bone1Name=dyn_skirtLow_sideB_Const_Base_r
  bone2Name=thigh_sub_r
  swing1LimitDegrees=45
  swing1Motion=Free
  swing2LimitDegrees=45
  swing2Motion=Free
  twistLimitDegrees=45
  twistMotion=Free
 }
 {
  bone1Name=dyn_skirtLow_sideB_Base_r
  bone2Name=dyn_skirtLow_sideB_Const_Base_r
  swing1LimitDegrees=45
  swing1Motion=Free
  swing2LimitDegrees=45
  swing2Motion=Free
  twistLimitDegrees=45
  twistMotion=Free
 }
 {bone1Name=dyn_skirtLow_sideB_r, bone2Name=dyn_skirtLow_sideB_Base_r, swing1LimitDegrees=15, swing2LimitDegrees=15, twistLimitDegrees=10}
 {
  bone1Name=dyn_skirtLow_sideA_Const_Base_r
  bone2Name=thigh_sub_r
  swing1LimitDegrees=45
  swing1Motion=Free
  swing2LimitDegrees=45
  swing2Motion=Free
  twistLimitDegrees=45
  twistMotion=Free
 }
 {
  bone1Name=dyn_skirtLow_sideA_Base_r
  bone2Name=dyn_skirtLow_sideA_Const_Base_r
  swing1LimitDegrees=45
  swing1Motion=Free
  swing2LimitDegrees=45
  swing2Motion=Free
  twistLimitDegrees=45
  twistMotion=Free
 }
 {
  bone1Name=dyn_skirtLow_frontB_Const_Base_r
  bone2Name=thigh_sub_r
  swing1LimitDegrees=45
  swing1Motion=Free
  swing2LimitDegrees=45
  swing2Motion=Free
  twistLimitDegrees=45
  twistMotion=Free
 }
 {
  bone1Name=dyn_skirtLow_frontB_Base_r
  bone2Name=dyn_skirtLow_frontB_Const_Base_r
  swing1LimitDegrees=45
  swing1Motion=Free
  swing2LimitDegrees=45
  swing2Motion=Free
  twistLimitDegrees=45
  twistMotion=Free
 }
 {bone1Name=dyn_skirtLow_frontB_r, bone2Name=dyn_skirtLow_frontB_Base_r, swing1LimitDegrees=15, swing2LimitDegrees=15, twistLimitDegrees=10}
 {
  bone1Name=dyn_skirtLow_frontA_Const_Base_r
  bone2Name=thigh_sub_r
  swing1LimitDegrees=45
  swing1Motion=Free
  swing2LimitDegrees=45
  swing2Motion=Free
  twistLimitDegrees=45
  twistMotion=Free
 }
 {
  bone1Name=dyn_skirtLow_frontA_Base_r
  bone2Name=dyn_skirtLow_frontA_Const_Base_r
  swing1LimitDegrees=45
  swing1Motion=Free
  swing2LimitDegrees=45
  swing2Motion=Free
  twistLimitDegrees=45
  twistMotion=Free
 }
 {
  bone1Name=dyn_skirtHi_backB_Base_r
  bone2Name=dyn_skirtHi_backB_Const_Base_r
  swing1LimitDegrees=45
  swing1Motion=Free
  swing2LimitDegrees=45
  swing2Motion=Free
  twistLimitDegrees=45
  twistMotion=Free
 }
 {
  bone1Name=dyn_skirtHi_backA_Base_r
  bone2Name=dyn_skirtHi_backA_Const_Base_r
  swing1LimitDegrees=45
  swing1Motion=Free
  swing2LimitDegrees=45
  swing2Motion=Free
  twistLimitDegrees=45
  twistMotion=Free
 }
 {bone1Name=dyn_skirtHi_backA_r, bone2Name=dyn_skirtHi_backA_Base_r, swing1LimitDegrees=20, swing2LimitDegrees=20, twistLimitDegrees=20}
 {
  bone1Name=dyn_skirtLow_backB_Base_r
  bone2Name=dyn_skirtLow_backB_Const_Base_r
  swing1LimitDegrees=45
  swing1Motion=Free
  swing2LimitDegrees=45
  swing2Motion=Free
  twistLimitDegrees=45
  twistMotion=Free
 }
 {bone1Name=dyn_skirtLow_backB_r, bone2Name=dyn_skirtLow_backB_Base_r, swing1LimitDegrees=20, swing2LimitDegrees=20, twistLimitDegrees=20}
 {
  bone1Name=dyn_skirtLow_backA_Base_r
  bone2Name=dyn_skirtLow_backA_Const_Base_r
  swing1LimitDegrees=45
  swing1Motion=Free
  swing2LimitDegrees=45
  swing2Motion=Free
  twistLimitDegrees=45
  twistMotion=Free
 }
 {
  bone1Name=dyn_skirtLow_backA_Base_l
  bone2Name=dyn_skirtLow_backA_Const_Base_l
  swing1LimitDegrees=45
  swing1Motion=Free
  swing2LimitDegrees=45
  swing2Motion=Free
  twistLimitDegrees=45
  twistMotion=Free
 }
 {
  bone1Name=dyn_skirtHi_backB_Base_l
  bone2Name=dyn_skirtHi_backB_Const_Base_l
  swing1LimitDegrees=45
  swing1Motion=Free
  swing2LimitDegrees=45
  swing2Motion=Free
  twistLimitDegrees=45
  twistMotion=Free
 }
 {bone1Name=dyn_skirtHi_backB_l, bone2Name=dyn_skirtHi_backB_Base_l, swing1LimitDegrees=20, swing2LimitDegrees=35, twistLimitDegrees=20}
 {
  bone1Name=dyn_skirtHi_backA_Base_l
  bone2Name=dyn_skirtHi_backA_Const_Base_l
  swing1LimitDegrees=45
  swing1Motion=Free
  swing2LimitDegrees=45
  swing2Motion=Free
  twistLimitDegrees=45
  twistMotion=Free
 }
 {
  bone1Name=dyn_skirtLow_frontA_Base_l
  bone2Name=dyn_skirtLow_frontA_Const_Base_l
  swing1LimitDegrees=45
  swing1Motion=Free
  swing2LimitDegrees=45
  swing2Motion=Free
  twistLimitDegrees=45
  twistMotion=Free
 }
 {bone1Name=dyn_skirtLow_frontA_l, bone2Name=dyn_skirtLow_frontA_Base_l, swing1LimitDegrees=15, swing2LimitDegrees=15, twistLimitDegrees=10}
 {
  bone1Name=dyn_skirtLow_frontB_Base_l
  bone2Name=dyn_skirtLow_frontB_Const_Base_l
  swing1LimitDegrees=45
  swing1Motion=Free
  swing2LimitDegrees=45
  swing2Motion=Free
  twistLimitDegrees=45
  twistMotion=Free
 }
 {
  bone1Name=dyn_skirtLow_sideA_Base_l
  bone2Name=dyn_skirtLow_sideA_Const_Base_l
  swing1LimitDegrees=45
  swing1Motion=Free
  swing2LimitDegrees=45
  swing2Motion=Free
  twistLimitDegrees=45
  twistMotion=Free
 }
 {bone1Name=dyn_skirtLow_sideA_l, bone2Name=dyn_skirtLow_sideA_Base_l, swing1LimitDegrees=15, swing2LimitDegrees=15, twistLimitDegrees=10}
 {
  bone1Name=dyn_skirtLow_sideB_Base_l
  bone2Name=dyn_skirtLow_sideB_Const_Base_l
  swing1LimitDegrees=45
  swing1Motion=Free
  swing2LimitDegrees=45
  swing2Motion=Free
  twistLimitDegrees=45
  twistMotion=Free
 }
 {bone1Name=dyn_skirtLow_sideB_l, bone2Name=dyn_skirtLow_sideB_Base_l, swing1LimitDegrees=15, swing2LimitDegrees=15, twistLimitDegrees=10}
 {
  bone1Name=dyn_skirtHi_sideB_Base_l
  bone2Name=dyn_skirtHi_sideB_Const_Base_l
  swing1LimitDegrees=45
  swing1Motion=Free
  swing2LimitDegrees=45
  swing2Motion=Free
  twistLimitDegrees=45
  twistMotion=Free
 }
 {bone1Name=dyn_skirtHi_sideB_l, bone2Name=dyn_skirtHi_sideB_Base_l, swing1LimitDegrees=15, swing2LimitDegrees=15, twistLimitDegrees=10}
 {
  bone1Name=dyn_skirtHi_sideA_Base_l
  bone2Name=dyn_skirtHi_sideA_Const_Base_l
  swing1LimitDegrees=45
  swing1Motion=Free
  swing2LimitDegrees=45
  swing2Motion=Free
  twistLimitDegrees=45
  twistMotion=Free
 }
 {
  bone1Name=dyn_skirtHi_frontB_Base_l
  bone2Name=dyn_skirtHi_frontB_Const_Base_l
  swing1LimitDegrees=45
  swing1Motion=Free
  swing2LimitDegrees=45
  swing2Motion=Free
  twistLimitDegrees=45
  twistMotion=Free
 }
 {bone1Name=dyn_skirtHi_frontB_l, bone2Name=dyn_skirtHi_frontB_Base_l, swing1LimitDegrees=15, swing2LimitDegrees=15, twistLimitDegrees=10}
 {
  bone1Name=dyn_skirtHi_frontA_Base_l
  bone2Name=dyn_skirtHi_frontA_Const_Base_l
  swing1LimitDegrees=45
  swing1Motion=Free
  swing2LimitDegrees=45
  swing2Motion=Free
  twistLimitDegrees=45
  twistMotion=Free
 }
 {bone1Name=dyn_necklace_02, bone2Name=dyn_necklace_01, swing1LimitDegrees=10, swing2LimitDegrees=15, twistLimitDegrees=5}
 {bone1Name=dyn_clamTop, bone2Name=dyn_necklace_02, swing1LimitDegrees=12, swing2LimitDegrees=12, twistLimitDegrees=12}
 {bone1Name=dyn_necklace_01, bone2Name=spine_05, swing1LimitDegrees=15, swing2LimitDegrees=15, twistLimitDegrees=45, twistMotion=Locked}
 {bone1Name=dyn_skirtHi_backB_r, bone2Name=dyn_skirtHi_backB_Base_r, swing1LimitDegrees=20, swing2LimitDegrees=20, twistLimitDegrees=20}
]
