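# Post-process style animation Blueprint for the Jellyfish female player body that adds secondary motion to its layered skirt.
# It copies the attached parent mesh's pose, then constrains skirt base bones to the pelvis, thighs and neighboring skirt bones.
# A crouch flag sets crouchAlpha each update, smoothly blending additive rotations on low front skirt bones.
# A final RigidBody node simulates the physics asset in root-bone space with custom gravity.
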AnimBlueprint /BRCosmetics/Characters/Player/Female/Medium/Bodies/F_MED_Jellyfish/Meshes/F_MED_Jellyfish_AnimBP
target_skeleton=&/F_MED_Jellyfish_Skeleton
graphs [
 AnimGraph {nodes [
   n1 = "Constraint - dyn_skirtHi_sideC_Const_Base_r (P)"(ComponentPose=Local To Component(LocalPose=Copy Pose From Mesh() {bUseAttachedParent=true}), Alpha=0.25, ConstraintWeights_0=1) {
    alphaBoolBlend {blendOption=Linear}
    boneToModify {boneName=dyn_skirtHi_sideC_Const_Base_r}
    constraintSetup [{targetBone {boneName=pelvis}}]
    constraintWeights [1]
   }
   n2 = "Constraint - dyn_skirtHi_backA_Const_Base_r (PP)"!(ComponentPose=@n1, Alpha=1, ConstraintWeights_0=0.7, ConstraintWeights_1:float=0.3) -> (Pose) {
    alphaBoolBlend {blendOption=Linear}
    boneToModify {boneName=dyn_skirtHi_backA_Const_Base_r}
    constraintSetup [{targetBone {boneName=dyn_skirtHi_sideC_r}}, {targetBone {boneName=dyn_skirtHi_backB_r}}]
    constraintWeights [1,1]
   }
   n3 = "Constraint - dyn_skirtLow_sideB_Const_Base_r (P)"(ComponentPose=@n2, Alpha=0.25, ConstraintWeights_0=1) {
    alphaBoolBlend {blendOption=Linear}
    boneToModify {boneName=dyn_skirtLow_sideB_Const_Base_r}
    constraintSetup [{targetBone {boneName=pelvis}}]
    constraintWeights [1]
   }
   n4 = "Constraint - dyn_skirtLow_backA_Const_Base_r (PP)"!(ComponentPose=@n3, Alpha=1, ConstraintWeights_0=0.3, ConstraintWeights_1:float=0.7) -> (Pose) {
    alphaBoolBlend {blendOption=Linear}
    boneToModify {boneName=dyn_skirtLow_backA_Const_Base_r}
    constraintSetup [{targetBone {boneName=dyn_skirtLow_backB_r}}, {targetBone {boneName=dyn_skirtLow_sideB_r}}]
    constraintWeights [1,1]
   }
   n5 = "Constraint - dyn_skirtHi_sideB_Const_Base_l (P)"(ComponentPose=@n4, Alpha=0.25, ConstraintWeights_0=1) {
    alphaBoolBlend {blendOption=Linear}
    boneToModify {boneName=dyn_skirtHi_sideB_Const_Base_l}
    constraintSetup [{targetBone {boneName=pelvis}}]
    constraintWeights [1]
   }
   n6 = "Constraint - dyn_skirtHi_backA_Const_Base_l (PP)"!(ComponentPose=@n5, Alpha=1, ConstraintWeights_0=0.7, ConstraintWeights_1:float=0.3) -> (Pose) {
    alphaBoolBlend {blendOption=Linear}
    boneToModify {boneName=dyn_skirtHi_backA_Const_Base_l}
    constraintSetup [{targetBone {boneName=dyn_skirtHi_sideB_l}}, {targetBone {boneName=dyn_skirtHi_backB_l}}]
    constraintWeights [1,1]
   }
   n7 = "Constraint - dyn_skirtLow_sideB_Const_Base_l (P)"(ComponentPose=@n6, Alpha=0.25, ConstraintWeights_0=1) {
    alphaBoolBlend {blendOption=Linear}
    boneToModify {boneName=dyn_skirtLow_sideB_Const_Base_l}
    constraintSetup [{targetBone {boneName=pelvis}}]
    constraintWeights [1]
   }
   n8 = "Constraint - dyn_skirtLow_backA_Const_Base_l (PP)"!(ComponentPose=@n7, Alpha=1, ConstraintWeights_0=0.3, ConstraintWeights_1:float=0.7) -> (Pose) {
    alphaBoolBlend {blendOption=Linear}
    boneToModify {boneName=dyn_skirtLow_backA_Const_Base_l}
    constraintSetup [{targetBone {boneName=dyn_skirtHi_backB_r}}, {targetBone {boneName=dyn_skirtLow_sideB_l}}]
    constraintWeights [1,1]
   }
   n9 = "Constraint - dyn_skirtHi_frontA_Const_Base_r (PP)"!(ComponentPose=@n8, Alpha=1, ConstraintWeights_0=0.75, ConstraintWeights_1:float=0.25) -> (Pose) {
    alphaBoolBlend {blendOption=Linear}
    boneToModify {boneName=dyn_skirtHi_frontA_Const_Base_r}
    constraintSetup [{targetBone {boneName=thigh_sub_r}}, {targetBone {boneName=dyn_skirtHi_frontA_l}}]
    constraintWeights [1,1]
   }
   n10 = "Constraint - dyn_skirtHi_frontA_Const_Base_l (PP)"!(ComponentPose=@n9, Alpha=1, ConstraintWeights_0=0.75, ConstraintWeights_1:float=0.25) -> (Pose) {
    alphaBoolBlend {blendOption=Linear}
    boneToModify {boneName=dyn_skirtHi_frontA_Const_Base_l}
    constraintSetup [{targetBone {boneName=thigh_sub_l}}, {targetBone {boneName=dyn_skirtHi_frontA_r}}]
    constraintWeights [1,1]
   }
   n11 = "Constraint - dyn_skirtLow_frontA_Const_Base_r (PP)"!(ComponentPose=@n10, Alpha=1, ConstraintWeights_0=0.8, ConstraintWeights_1:float=0.2) -> (Pose) {
    alphaBoolBlend {blendOption=Linear}
    boneToModify {boneName=dyn_skirtLow_frontA_Const_Base_r}
    constraintSetup [{targetBone {boneName=thigh_sub_r}}, {targetBone {boneName=dyn_skirtLow_frontA_l}}]
    constraintWeights [1,1]
   }
   n12 = "Constraint - dyn_skirtLow_frontA_Const_Base_l (PP)"!(ComponentPose=@n11, Alpha=1, ConstraintWeights_0=0.8, ConstraintWeights_1:float=0.2) -> (Pose) {
    alphaBoolBlend {blendOption=Linear}
    boneToModify {boneName=dyn_skirtLow_frontA_Const_Base_l}
    constraintSetup [{targetBone {boneName=thigh_sub_l}}, {targetBone {boneName=dyn_skirtLow_frontA_r}}]
    constraintWeights [1,1]
   }
   n13 = Get crouchAlpha() -> (crouchAlpha:double)
   n14 = "Transform (Modify) Bone - Bone: dyn_skirtLow_frontA_Base_l"(ComponentPose=@n12, Alpha=@n13, Translation="0,0,0", Rotation="0.000,-8,0.000", Scale="1,1,1") {
    rotationMode=BMM_Additive
    rotationSpace=BCS_BoneSpace
    translationMode=BMM_Additive
    translationSpace=BCS_BoneSpace
    alphaBoolBlend {blendOption=Linear}
    alphaScaleBiasClamp {bInterpResult=true, interpSpeedDecreasing=5, interpSpeedIncreasing=5}
    boneToModify {boneName=dyn_skirtLow_frontA_Base_l}
    rotation {yaw=-8}
   }
   n15 = "Transform (Modify) Bone - Bone: dyn_skirtLow_frontB_Base_l"(ComponentPose=@n14, Alpha=@n13, Translation="0,0,0", Rotation="0.000,-8,0.000", Scale="1,1,1") {
    rotationMode=BMM_Additive
    rotationSpace=BCS_BoneSpace
    translationMode=BMM_Additive
    translationSpace=BCS_BoneSpace
    alphaBoolBlend {blendOption=Linear}
    alphaScaleBiasClamp {bInterpResult=true, interpSpeedDecreasing=5, interpSpeedIncreasing=5}
    boneToModify {boneName=dyn_skirtLow_frontB_Base_l}
    rotation {yaw=-8}
   }
   n16 = "Transform (Modify) Bone - Bone: dyn_skirtLow_frontA_Base_r"(ComponentPose=@n15, Alpha=@n13, Translation="0,0,0", Rotation="-8,-8,0", Scale="1,1,1") {
    rotationMode=BMM_Additive
    rotationSpace=BCS_BoneSpace
    translationMode=BMM_Additive
    translationSpace=BCS_BoneSpace
    alphaBoolBlend {blendOption=Linear}
    alphaScaleBiasClamp {bInterpResult=true, interpSpeedDecreasing=5, interpSpeedIncreasing=5}
    boneToModify {boneName=dyn_skirtLow_frontA_Base_r}
    rotation {pitch=-8, yaw=-8}
   }
   n17 = "Transform (Modify) Bone - Bone: dyn_skirtLow_frontB_Base_r"(ComponentPose=@n16, Alpha=@n13, Translation="0,0,0", Rotation="-6,-8,0.000", Scale="1,1,1") {
    rotationMode=BMM_Additive
    rotationSpace=BCS_BoneSpace
    translationMode=BMM_Additive
    translationSpace=BCS_BoneSpace
    alphaBoolBlend {blendOption=Linear}
    alphaScaleBiasClamp {bInterpResult=true, interpSpeedDecreasing=5, interpSpeedIncreasing=5}
    boneToModify {boneName=dyn_skirtLow_frontB_Base_r}
    rotation {pitch=-6, yaw=-8}
   }
   n18 = RigidBody(ComponentPose=@n17) {
    bClampLinearTranslationLimitToRefPose=true
    bOverrideWorldGravity=true
    overridePhysicsAsset=&/F_MED_Jellyfish_Physics
    simulationSpace=BaseBoneSpace
    alphaBoolBlend {blendOption=Linear}
    baseBoneRef {boneName=root}
    componentAppliedLinearAccClamp=<5000 5000 5000>
    componentLinearAccScale=<2 2 0.5>
    componentLinearVelScale=<0.5 0.5 0.4>
    overrideWorldGravity {z=-600}
   }
   Output Pose(Result=Component To Local(ComponentPose=@n18)) {layerGroup=None}
  ]}
 EventGraph {nodes [
   Event BlueprintUpdateAnimation()
   > Set crouchAlpha(crouchAlpha:double="To Float (Boolean)"(InBool=Get bIsCrouching())) -> (Output_Get:double)
   TryGetPawnOwner()
  ]}
]
variables [crouchAlpha {type=number}]
