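# Control Rig for the Jellyfish female medium character body that procedurally sways the eight head hair-tendril bone chains (B to E, left and right).
# A sequence of Chain Harmonics units runs from Begin Execution, one per chain, using only Wave motion with Pendulum and Reach disabled.
# Per-chain speed, frequency, amplitude, noise and easing variations diversify the motion.
# B and C chains start at bone 01, D and E at 02; debug drawing is enabled throughout.
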ControlRigBlueprint /BRCosmetics/Characters/Player/Female/Medium/Bodies/F_MED_Jellyfish/Parts/F_MED_JellyfishTendril_ControlRig
graphs [
 RigVMFunctionLibrary {links [], nodes []}
 RigVMModel {nodes [
   n1 = &RigUnit_BeginExecution()
   /Script/ControlRig.RigUnit_ChainHarmonics(
    ChainRoot=dyn_hairTrndrilB_01_l
    DrawWorldOffset=(Rotation=(X=0,Y=0,Z=0,W=1),Translation=(X=0,Y=0,Z=0),Scale3D=(X=1,Y=1,Z=1))
    Pendulum=(bEnabled=False,PendulumStiffness=2,PendulumGravity=(X=0,Y=0,Z=0),PendulumBlend=0.75,PendulumDrag=0.98,PendulumMinimum=0,PendulumMaximum=0.1,PendulumEase=Linear,UnwindAxis=(X=0,Y=1,Z=0),UnwindMinimum=0.2,UnwindMaximum=0.05)
    Reach=(bEnabled=False,ReachTarget=(X=0,Y=0,Z=0),ReachAxis=(X=1,Y=0,Z=0),ReachMinimum=0,ReachMaximum=0,ReachEase=SineEaseInOut)
    Speed=(X=1.1,Y=0.6,Z=0.45)
    Wave=(bEnabled=True,WaveFrequency=(X=1,Y=0.6,Z=0.8),WaveAmplitude=(X=0.2,Y=0.48,Z=0.185),WaveOffset=(X=0,Y=1,Z=2),WaveNoise=(X=2,Y=0.5,Z=0.6),WaveMinimum=0,WaveMaximum=0.75,WaveEase=SineEaseInOut)
    WaveCurve=(EditorCurveData=(Keys=((Time=-0.0198,Value=0.4929),(Time=0.9957,Value=0.9415)),PreInfinityExtrap=RCCE_Constant,PostInfinityExtrap=RCCE_Constant,DefaultValue=340282346638528859811704183484516925440))
    bDrawDebug=True
    ExecutePin=@n1.ExecutePin
   ) {is=struct RigVMUnitNode}
   > &RigUnit_ChainHarmonics(
    ChainRoot=dyn_hairTrndrilC_01_l
    DrawWorldOffset=(Rotation=(X=0,Y=0,Z=0,W=1),Translation=(X=0,Y=0,Z=0),Scale3D=(X=1,Y=1,Z=1))
    Pendulum=(bEnabled=False,PendulumStiffness=2,PendulumGravity=(X=0,Y=0,Z=0),PendulumBlend=0.75,PendulumDrag=0.98,PendulumMinimum=0,PendulumMaximum=0.1,PendulumEase=Linear,UnwindAxis=(X=0,Y=1,Z=0),UnwindMinimum=0.2,UnwindMaximum=0.05)
    Reach=(bEnabled=False,ReachTarget=(X=0,Y=0,Z=0),ReachAxis=(X=1,Y=0,Z=0),ReachMinimum=0,ReachMaximum=0,ReachEase=SineEaseInOut)
    Speed=(X=0.2,Y=1.1,Z=0.65)
    Wave=(bEnabled=True,WaveFrequency=(X=0.6,Y=1,Z=0.8),WaveAmplitude=(X=0.2,Y=0.48,Z=0.22),WaveOffset=(X=0,Y=1,Z=2),WaveNoise=(X=0.8,Y=1.5,Z=2),WaveMinimum=0,WaveMaximum=0.75,WaveEase=SineEaseInOut)
    WaveCurve=(EditorCurveData=(Keys=((Time=-0.0198,Value=0.4929),(Time=0.9957,Value=0.9415)),PreInfinityExtrap=RCCE_Constant,PostInfinityExtrap=RCCE_Constant,DefaultValue=340282346638528859811704183484516925440))
    bDrawDebug=True
   )
   > &RigUnit_ChainHarmonics(
    ChainRoot=dyn_hairTrndrilD_02_l
    DrawWorldOffset=(Rotation=(X=0,Y=0,Z=0,W=1),Translation=(X=0,Y=0,Z=0),Scale3D=(X=1,Y=1,Z=1))
    Pendulum=(bEnabled=False,PendulumStiffness=2,PendulumGravity=(X=0,Y=0,Z=0),PendulumBlend=0.75,PendulumDrag=0.98,PendulumMinimum=0,PendulumMaximum=0.1,PendulumEase=Linear,UnwindAxis=(X=0,Y=1,Z=0),UnwindMinimum=0.2,UnwindMaximum=0.05)
    Reach=(bEnabled=False,ReachTarget=(X=0,Y=0,Z=0),ReachAxis=(X=1,Y=0,Z=0),ReachMinimum=0,ReachMaximum=0,ReachEase=SineEaseInOut)
    Speed=(X=0.6,Y=0.85,Z=0.18)
    Wave=(bEnabled=True,WaveFrequency=(X=0.5,Y=1.2,Z=0.95),WaveAmplitude=(X=0.2,Y=0.48,Z=0.15),WaveOffset=(X=0,Y=1,Z=2),WaveNoise=(X=2.3,Y=1.8,Z=0.85),WaveMinimum=0,WaveMaximum=0.75,WaveEase=CircularEaseInOut)
    WaveCurve=(EditorCurveData=(Keys=((Time=-0.0198,Value=0.4929),(Time=0.9957,Value=0.9415)),PreInfinityExtrap=RCCE_Constant,PostInfinityExtrap=RCCE_Constant,DefaultValue=340282346638528859811704183484516925440))
    bDrawDebug=True
   )
   > &RigUnit_ChainHarmonics(
    ChainRoot=dyn_hairTrndrilE_02_l
    DrawWorldOffset=(Rotation=(X=0,Y=0,Z=0,W=1),Translation=(X=0,Y=0,Z=0),Scale3D=(X=1,Y=1,Z=1))
    Pendulum=(bEnabled=False,PendulumStiffness=2,PendulumGravity=(X=0,Y=0,Z=0),PendulumBlend=0.75,PendulumDrag=0.98,PendulumMinimum=0,PendulumMaximum=0.1,PendulumEase=Linear,UnwindAxis=(X=0,Y=1,Z=0),UnwindMinimum=0.2,UnwindMaximum=0.05)
    Reach=(bEnabled=False,ReachTarget=(X=0,Y=0,Z=0),ReachAxis=(X=1,Y=0,Z=0),ReachMinimum=0,ReachMaximum=0,ReachEase=SineEaseInOut)
    Speed=(X=1.1,Y=0.125,Z=1.1)
    Wave=(bEnabled=True,WaveFrequency=(X=1,Y=0.6,Z=0.8),WaveAmplitude=(X=0.2,Y=0.48,Z=0.235),WaveOffset=(X=0,Y=1,Z=2),WaveNoise=(X=2,Y=0.5,Z=0.6),WaveMinimum=0,WaveMaximum=0.75,WaveEase=SineEaseInOut)
    WaveCurve=(EditorCurveData=(Keys=((Time=-0.0198,Value=0.4929),(Time=0.9957,Value=0.9415)),PreInfinityExtrap=RCCE_Constant,PostInfinityExtrap=RCCE_Constant,DefaultValue=340282346638528859811704183484516925440))
    bDrawDebug=True
   )
   > &RigUnit_ChainHarmonics(
    ChainRoot=dyn_hairTrndrilC_01_r
    DrawWorldOffset=(Rotation=(X=0,Y=0,Z=0,W=1),Translation=(X=0,Y=0,Z=0),Scale3D=(X=1,Y=1,Z=1))
    Pendulum=(bEnabled=False,PendulumStiffness=2,PendulumGravity=(X=0,Y=0,Z=0),PendulumBlend=0.75,PendulumDrag=0.98,PendulumMinimum=0,PendulumMaximum=0.1,PendulumEase=Linear,UnwindAxis=(X=0,Y=1,Z=0),UnwindMinimum=0.2,UnwindMaximum=0.05)
    Reach=(bEnabled=False,ReachTarget=(X=0,Y=0,Z=0),ReachAxis=(X=1,Y=0,Z=0),ReachMinimum=0,ReachMaximum=0,ReachEase=SineEaseInOut)
    Speed=(X=0.85,Y=0.6,Z=0.5)
    Wave=(bEnabled=True,WaveFrequency=(X=1,Y=0.6,Z=0.8),WaveAmplitude=(X=0.2,Y=0.48,Z=0.05),WaveOffset=(X=0,Y=1,Z=2),WaveNoise=(X=2,Y=0.5,Z=0.6),WaveMinimum=0,WaveMaximum=0.75,WaveEase=QuarticEaseInOut)
    WaveCurve=(EditorCurveData=(Keys=((Time=-0.0198,Value=0.4929),(Time=0.9957,Value=0.9415)),PreInfinityExtrap=RCCE_Constant,PostInfinityExtrap=RCCE_Constant,DefaultValue=340282346638528859811704183484516925440))
    bDrawDebug=True
   )
   > &RigUnit_ChainHarmonics(
    ChainRoot=dyn_hairTrndrilE_02_r
    DrawWorldOffset=(Rotation=(X=0,Y=0,Z=0,W=1),Translation=(X=0,Y=0,Z=0),Scale3D=(X=1,Y=1,Z=1))
    Pendulum=(bEnabled=False,PendulumStiffness=2,PendulumGravity=(X=0,Y=0,Z=0),PendulumBlend=0.75,PendulumDrag=0.98,PendulumMinimum=0,PendulumMaximum=0.1,PendulumEase=Linear,UnwindAxis=(X=0,Y=1,Z=0),UnwindMinimum=0.2,UnwindMaximum=0.05)
    Reach=(bEnabled=False,ReachTarget=(X=0,Y=0,Z=0),ReachAxis=(X=1,Y=0,Z=0),ReachMinimum=0,ReachMaximum=0,ReachEase=SineEaseInOut)
    Speed=(X=0.275,Y=1.1,Z=0.75)
    Wave=(bEnabled=True,WaveFrequency=(X=0.6,Y=1,Z=0.8),WaveAmplitude=(X=0.2,Y=0.48,Z=0.22),WaveOffset=(X=0,Y=1,Z=2),WaveNoise=(X=0.8,Y=1.5,Z=2),WaveMinimum=0,WaveMaximum=0.758,WaveEase=QuadraticEaseInOut)
    WaveCurve=(EditorCurveData=(Keys=((Time=-0.0198,Value=0.4929),(Time=0.9957,Value=0.9415)),PreInfinityExtrap=RCCE_Constant,PostInfinityExtrap=RCCE_Constant,DefaultValue=340282346638528859811704183484516925440))
    bDrawDebug=True
   )
   > &RigUnit_ChainHarmonics(
    ChainRoot=dyn_hairTrndrilB_01_r
    DrawWorldOffset=(Rotation=(X=0,Y=0,Z=0,W=1),Translation=(X=0,Y=0,Z=0),Scale3D=(X=1,Y=1,Z=1))
    Pendulum=(bEnabled=False,PendulumStiffness=2,PendulumGravity=(X=0,Y=0,Z=0),PendulumBlend=0.75,PendulumDrag=0.98,PendulumMinimum=0,PendulumMaximum=0.1,PendulumEase=Linear,UnwindAxis=(X=0,Y=1,Z=0),UnwindMinimum=0.2,UnwindMaximum=0.05)
    Reach=(bEnabled=False,ReachTarget=(X=0,Y=0,Z=0),ReachAxis=(X=1,Y=0,Z=0),ReachMinimum=0,ReachMaximum=0,ReachEase=SineEaseInOut)
    Speed=(X=0.6,Y=0.85,Z=0.2)
    Wave=(bEnabled=True,WaveFrequency=(X=0.5,Y=1.2,Z=0.95),WaveAmplitude=(X=0.2,Y=0.48,Z=0.06),WaveOffset=(X=0,Y=1,Z=2),WaveNoise=(X=2.3,Y=1.8,Z=0.85),WaveMinimum=0,WaveMaximum=0.6,WaveEase=BackEaseInOut)
    WaveCurve=(EditorCurveData=(Keys=((Time=-0.0198,Value=0.4929),(Time=0.9957,Value=0.9415)),PreInfinityExtrap=RCCE_Constant,PostInfinityExtrap=RCCE_Constant,DefaultValue=340282346638528859811704183484516925440))
    bDrawDebug=True
   )
   > &RigUnit_ChainHarmonics(
    ChainRoot=dyn_hairTrndrilD_02_r
    DrawWorldOffset=(Rotation=(X=0,Y=0,Z=0,W=1),Translation=(X=0,Y=0,Z=0),Scale3D=(X=1,Y=1,Z=1))
    Pendulum=(bEnabled=False,PendulumStiffness=2,PendulumGravity=(X=0,Y=0,Z=0),PendulumBlend=0.75,PendulumDrag=0.98,PendulumMinimum=0,PendulumMaximum=0.1,PendulumEase=Linear,UnwindAxis=(X=0,Y=1,Z=0),UnwindMinimum=0.2,UnwindMaximum=0.05)
    Reach=(bEnabled=False,ReachTarget=(X=0,Y=0,Z=0),ReachAxis=(X=1,Y=0,Z=0),ReachMinimum=0,ReachMaximum=0,ReachEase=SineEaseInOut)
    Speed=(X=1.2,Y=0.125,Z=1.12)
    Wave=(bEnabled=True,WaveFrequency=(X=1,Y=0.6,Z=0.8),WaveAmplitude=(X=0.2,Y=0.48,Z=0.235),WaveOffset=(X=0,Y=1,Z=2),WaveNoise=(X=2,Y=0.5,Z=0.6),WaveMinimum=0,WaveMaximum=0.75,WaveEase=SineEaseInOut)
    WaveCurve=(EditorCurveData=(Keys=((Time=-0.0198,Value=0.4929),(Time=0.9957,Value=0.9415)),PreInfinityExtrap=RCCE_Constant,PostInfinityExtrap=RCCE_Constant,DefaultValue=340282346638528859811704183484516925440))
    bDrawDebug=True
   )
  ]}
]
hierarchy [
 root {loc=<0 0 0>, rot=<0 0 0 1>}
 attach {parent=root, loc=<0 0 0>, rot=<0 0 0 1>}
 pelvis {parent=root, loc=<0 -0.5301 98.0571>, rot=<0.0127 0.707 -0.0127 -0.707>}
 spine_01 {parent=pelvis, loc=<6.1208 -0.9328 0>, rot=<0 0 0.034 -0.9994>}
 spine_02 {parent=spine_01, loc=<6.5996 -0.2095 0>, rot=<0 0 0.0238 -0.9997>}
 spine_03 {parent=spine_02, loc=<6.4914 -0.1237 0>, rot=<0 0 0.0149 -0.9999>}
 spine_04 {parent=spine_03, loc=<6.9598 0.0794 0>, rot=<0 0 -0.0132 -0.9999>}
 spine_05 {parent=spine_04, loc=<6.9987 0 0>, rot=<0 0 -0.0015 -1>}
 clavicle_l {parent=spine_05, loc=<5.4831 -0.1053 -3.2946>, rot=<0.0465 -0.7289 -0.009 -0.6829>}
 clavicle_r {parent=spine_05, loc=<5.4831 -0.1053 3.2946>, rot=<0.7289 0.0465 -0.6829 0.009>}
 neck_01 {parent=spine_05, loc=<11.3138 0.1529 0>, rot=<0 0 -0.2003 -0.9797>}
 neck_02 {parent=neck_01, loc=<4.5627 0 0>, rot=<0 0 0.0504 -0.9987>}
 head {parent=neck_02, loc=<5.3914 0 0>, rot=<0 0 0.115 -0.9934>}
 dyn_earring_l {parent=head, loc=<0.236 -0.0572 -6.4098>, rot=<0.1659 -0.6471 0.0625 -0.7415>}
 dyn_earring_r {parent=head, loc=<0.2363 -0.0572 6.4098>, rot=<-0.6471 -0.1659 0.7415 0.0625>}
 faceAttach {parent=head, loc=<-1.3448 -0.1103 0.0058>, rot=<0.0149 -0.3911 -0.0185 -0.92>, scale=<0.89 0.89 0.89>}
 C_brow_mid {parent=faceAttach, loc=<4.1334 9.173 4.4531>, rot=<0.0082 -0.2888 -0.0317 -0.9568>}
 L_brow_mid {parent=faceAttach, loc=<8.3577 8.9276 3.9518>, rot=<0.0663 -0.3901 0.2384 -0.8869>}
 L_brow_outer {parent=faceAttach, loc=<10.16 6.1918 1.9778>, rot=<0.1291 -0.3808 0.3766 -0.8346>}
 L_cheek_inner {parent=faceAttach, loc=<4.8926 7.8324 -2.1689>, rot=<0.0421 -0.6158 0.353 -0.7032>}
 C_lip_upper_mid {parent=faceAttach, loc=<-0.3325 9.991 -0.1978>, rot=<0.2159 -0.3573 -0.0662 -0.9063>}
 L_lip_upper_outer {parent=faceAttach, loc=<0.6622 9.5156 -1.2844>, rot=<0.0764 -0.3521 0.2617 -0.8954>}
 L_lip_corner {parent=faceAttach, loc=<1.696 8.4926 -2.4125>, rot=<-0.0014 -0.3041 0.0126 -0.9526>}
 C_nose_bridge {parent=faceAttach, loc=<2.5933 10.9162 2.8869>, rot=<-0.244 -0.3688 0.0868 -0.8927>}
 R_brow_mid {parent=faceAttach, loc=<3.4738 8.8832 8.6654>, rot=<-0.1165 -0.3568 -0.2296 -0.898>}
 R_brow_outer {parent=faceAttach, loc=<1.6301 6.1584 10.4669>, rot=<-0.1527 -0.329 -0.3747 -0.8533>}
 R_cheek_inner {parent=faceAttach, loc=<-2.4861 7.6997 4.9367>, rot=<-0.2298 -0.0333 -0.2904 -0.9283>}
 R_lip_corner {parent=faceAttach, loc=<-2.6212 8.4142 1.7459>, rot=<-0.0138 -0.4427 -0.0214 -0.8963>}
 C_jaw {parent=faceAttach, loc=<1.1624 1.8735 1.2353>, rot=<0.0158 -0.3734 -0.0134 -0.9274>}
 L_lip_lower_outer {parent=C_jaw, loc=<1.4467 7.6814 -1.9928>, rot=<0.0099 -0.0745 0.2748 -0.9586>}
 C_lip_lower_mid {parent=C_jaw, loc=<0.0035 8.3096 -1.823>, rot=<0.0243 0.006 0.0019 -0.9997>}
 R_lip_lower_outer {parent=C_jaw, loc=<-1.5419 7.817 -1.9031>, rot=<0.0083 0.0211 -0.3267 -0.9449>}
 teeth_lower {parent=C_jaw, loc=<-0.0423 5.375 -4.3449>, rot=<-0.0351 -0.0109 0.0215 -0.9991>}
 tongue {parent=C_jaw, loc=<0.005 1.0955 -3.0077>, rot=<-0.183 0.0044 0.0039 -0.9831>}
 R_eye {parent=faceAttach, loc=<1.8452 6.4067 6.5945>, rot=<0.379 0.0177 0.9211 -0.0872>}
 R_eye_lid_upper_mid {parent=R_eye, loc=<-0.0492 -1.1692 0.2707>, rot=<0.0551 -0.001 0.0208 -0.9983>, scale=<1 1.0001 0.9999>}
 R_eye_lid_lower_mid {parent=R_eye, loc=<-0.016 -0.9427 -0.7033>, rot=<0.0334 -0.0005 0.0208 -0.9992>}
 L_eye {parent=faceAttach, loc=<6.4327 6.5 2.1838>, rot=<0.3647 0.0649 0.9283 0.0322>}
 L_eye_lid_upper_mid {parent=L_eye, loc=<0.0251 -1.1623 0.2388>, rot=<0.0523 0 0 -0.9986>}
 L_eye_lid_lower_mid {parent=L_eye, loc=<0.0223 -0.923 -0.7488>, rot=<0.0329 0 0 -0.9995>}
 teeth_upper {parent=faceAttach, loc=<-0.5528 8.1413 -0.3995>, rot=<-0.1 -0.3554 0.0555 -0.9277>}
 R_lip_upper_outer {parent=faceAttach, loc=<-1.5546 9.5279 0.8545>, rot=<-0.1065 -0.3265 -0.2976 -0.8908>}
 dyn_hairTrndrilE_01_Base_l {parent=head, loc=<11.6783 -5.3397 -3.5286>, rot=<0.5615 -0.792 -0.2206 0.0938>}
 dyn_hairTrndrilE_01_l {parent=dyn_hairTrndrilE_01_Base_l, loc=<0 0 0>, rot=<0 0 0 -1>}
 dyn_hairTrndrilE_02_l {parent=dyn_hairTrndrilE_01_l, loc=<10.1625 0 0>, rot=<-0.0252 -0.0269 -0.0744 -0.9965>}
 dyn_hairTrndrilE_03_l {parent=dyn_hairTrndrilE_02_l, loc=<8.5172 0 0>, rot=<-0.1924 0.0397 0.064 -0.9784>}
 dyn_hairTrndrilE_04_l {parent=dyn_hairTrndrilE_03_l, loc=<9.2482 0 0>, rot=<0.0888 -0.0958 -0.0345 -0.9908>}
 dyn_hairTrndrilE_05_l {parent=dyn_hairTrndrilE_04_l, loc=<8.7779 0 0>, rot=<-0.0846 0.1001 -0.0021 -0.9914>}
 dyn_hairTrndrilE_06_l {parent=dyn_hairTrndrilE_05_l, loc=<9.8581 0 0>, rot=<0 0 0 -1>}
 dyn_hairTrndrilD_01_Base_l {parent=head, loc=<11.0955 -4.5696 -4.7303>, rot=<0.4956 -0.8434 -0.1799 -0.1037>}
 dyn_hairTrndrilD_01_l {parent=dyn_hairTrndrilD_01_Base_l, loc=<0 0 0>, rot=<0 0 0 -1>}
 dyn_hairTrndrilD_02_l {parent=dyn_hairTrndrilD_01_l, loc=<12.0893 0 0>, rot=<0.087 -0.0012 0.0109 -0.9961>}
 dyn_hairTrndrilD_03_l {parent=dyn_hairTrndrilD_02_l, loc=<9.1331 0 0>, rot=<0.0052 0.0107 -0.0068 -0.9999>}
 dyn_hairTrndrilD_04_l {parent=dyn_hairTrndrilD_03_l, loc=<8.421 0 0>, rot=<-0.0019 -0.0801 -0.0145 -0.9967>}
 dyn_hairTrndrilD_05_l {parent=dyn_hairTrndrilD_04_l, loc=<6.4445 0 0>, rot=<0.0026 0.0814 0.012 -0.9966>}
 dyn_hairTrndrilD_06_l {parent=dyn_hairTrndrilD_05_l, loc=<8.9978 0 0>, rot=<-0.0872 0 0 -0.9962>}
 dyn_hairTrndrilC_01_Base_l {parent=head, loc=<10.5283 -4.5702 -5.9971>, rot=<0.3803 -0.8122 -0.4021 -0.1842>}
 dyn_hairTrndrilC_01_l {parent=dyn_hairTrndrilC_01_Base_l, loc=<0 0 0>, rot=<0 0 0 -1>}
 dyn_hairTrndrilC_02_l {parent=dyn_hairTrndrilC_01_l, loc=<15.328 0 0>, rot=<0.0156 0.0189 0.0138 -0.9996>}
 dyn_hairTrndrilC_03_l {parent=dyn_hairTrndrilC_02_l, loc=<15.4224 0 0>, rot=<-0.0221 -0.0275 -0.0006 -0.9994>}
 dyn_hairTrndrilC_04_l {parent=dyn_hairTrndrilC_03_l, loc=<9.3871 0 0>, rot=<0.0229 0.0272 -0.0239 -0.9991>}
 dyn_hairTrndrilC_05_l {parent=dyn_hairTrndrilC_04_l, loc=<9.5124 0 0>, rot=<0 0 0 -1>}
 dyn_hairTrndrilB_01_Base_l {parent=head, loc=<11.2915 -2.7348 -5.3268>, rot=<0.3108 -0.7257 -0.5662 -0.237>}
 dyn_hairTrndrilB_01_l {parent=dyn_hairTrndrilB_01_Base_l, loc=<0 0 0>, rot=<0 0 0 -1>}
 dyn_hairTrndrilB_02_l {parent=dyn_hairTrndrilB_01_l, loc=<12.7664 0 0>, rot=<0.0087 0.0094 0.0229 -0.9997>}
 dyn_hairTrndrilB_03_l {parent=dyn_hairTrndrilB_02_l, loc=<8.958 0.1198 -0.001>, rot=<-0.0615 -0.0622 -0.0557 -0.9946>}
 dyn_hairTrndrilB_04_l {parent=dyn_hairTrndrilB_03_l, loc=<9.2331 -0.5364 1.0011>, rot=<0.0976 0.0972 0.0478 -0.9893>}
 dyn_hairTrndrilB_05_l {parent=dyn_hairTrndrilB_04_l, loc=<6.7538 0.5265 -1.764>, rot=<-0.0301 -0.0324 -0.0106 -0.999>}
 dyn_hairTrndrilB_06_l {parent=dyn_hairTrndrilB_05_l, loc=<7.6071 -0.0968 0.411>, rot=<0 0 0 1>}
 dyn_hairTemple_01_Base_r {parent=head, loc=<-4.6297 0.1015 7.7854>, rot=<-0.0604 -0.5667 0.8135 0.1163>}
 dyn_hairTemple_01_r {parent=dyn_hairTemple_01_Base_r, loc=<0 0 0>, rot=<0 0 0 1>}
 dyn_hairTemple_02_r {parent=dyn_hairTemple_01_r, loc=<4.941 0 0>, rot=<-0.2202 0.0467 0.155 -0.9619>}
 dyn_hairTemple_03_r {parent=dyn_hairTemple_02_r, loc=<5.2634 0 0>, rot=<-0.3106 -0.0633 -0.0952 -0.9436>}
 dyn_hairTemple_01_Base_l {parent=head, loc=<-1.4512 0.1277 -7.7854>, rot=<-0.0087 0.5735 0.8191 0.0014>}
 dyn_hairTemple_01_l {parent=dyn_hairTemple_01_Base_l, loc=<0 0 0>, rot=<0 0 0 1>}
 dyn_hairTemple_02_l {parent=dyn_hairTemple_01_l, loc=<5.1253 0 0>, rot=<0.4354 -0.0702 0.253 -0.8611>}
 dyn_hairTemple_03_l {parent=dyn_hairTemple_02_l, loc=<7.1274 0 0>, rot=<0.0937 0.2432 -0.0072 -0.9654>}
 dyn_hairBack_01_Base_l {parent=head, loc=<1.9286 -4.2875 -5.7011>, rot=<0.0288 -0.9995 0.0002 -0.0086>}
 dyn_hairBack_01_l {parent=dyn_hairBack_01_Base_l, loc=<0 0 0>, rot=<0 0 0 -1>}
 dyn_hairBack_02_l {parent=dyn_hairBack_01_l, loc=<7.8815 0 0>, rot=<-0.0277 0.1434 0.1395 -0.9794>}
 dyn_hairBack_03_l {parent=dyn_hairBack_02_l, loc=<6.0784 0 0>, rot=<0.0257 -0.0722 0.0098 -0.997>}
 dyn_hairBack_01_c_Base {parent=head, loc=<0.591 -5.5907 -0.0581>, rot=<0.0155 -0.9999 0 -0.0039>}
 dyn_hairBack_01_c {parent=dyn_hairBack_01_c_Base, loc=<0 0 0>, rot=<0 0 0 -1>}
 dyn_hairBack_02_c {parent=dyn_hairBack_01_c, loc=<7.7812 0 0>, rot=<-0.0024 0.0145 0.1407 -0.9899>}
 dyn_hairBack_03_c {parent=dyn_hairBack_02_c, loc=<7.032 0 0>, rot=<-0.031 0.1122 -0.0354 -0.9926>}
 dyn_hairBack_01_Base_r {parent=head, loc=<1.4944 -4.7128 4.9508>, rot=<-0.0215 -0.995 0.0029 0.0978>}
 dyn_hairBack_01_r {parent=dyn_hairBack_01_Base_r, loc=<0 0 0>, rot=<0 0 0 -1>}
 dyn_hairBack_02_r {parent=dyn_hairBack_01_r, loc=<8.8354 0 0>, rot=<-0.0133 0.0444 0.3355 -0.9409>}
 dyn_hairBack_03_r {parent=dyn_hairBack_02_r, loc=<7.3453 0 0>, rot=<-0.0128 0.033 -0.2596 -0.9651>}
 dyn_hairTrndrilB_01_Base_r {parent=head, loc=<11.2918 -2.7348 5.3268>, rot=<0.3108 -0.7257 0.5663 0.237>}
 dyn_hairTrndrilB_01_r {parent=dyn_hairTrndrilB_01_Base_r, loc=<0 0 0>, rot=<0 0 0 1>}
 dyn_hairTrndrilB_02_r {parent=dyn_hairTrndrilB_01_r, loc=<12.7665 0 0>, rot=<-0.0087 -0.0092 0.0162 -0.9998>}
 dyn_hairTrndrilB_03_r {parent=dyn_hairTrndrilB_02_r, loc=<8.9591 0 0>, rot=<0.0083 0.0088 -0.0173 -0.9998>}
 dyn_hairTrndrilB_04_r {parent=dyn_hairTrndrilB_03_r, loc=<9.3026 0 0>, rot=<0.0833 0.0846 -0.0055 -0.9929>}
 dyn_hairTrndrilB_05_r {parent=dyn_hairTrndrilB_04_r, loc=<6.9998 0 0>, rot=<-0.12 -0.1251 0.0181 -0.9847>}
 dyn_hairTrndrilB_06_r {parent=dyn_hairTrndrilB_05_r, loc=<7.619 0 0>, rot=<0 0 0 1>}
 dyn_hairTrndrilC_01_Base_r {parent=head, loc=<10.5279 -4.5702 5.9971>, rot=<0.3803 -0.8122 0.4021 0.1842>}
 dyn_hairTrndrilC_01_r {parent=dyn_hairTrndrilC_01_Base_r, loc=<0 0 0>, rot=<0 0 0 -1>}
 dyn_hairTrndrilC_02_r {parent=dyn_hairTrndrilC_01_r, loc=<15.328 0 0>, rot=<-0.0156 -0.0189 0.0138 -0.9996>}
 dyn_hairTrndrilC_03_r {parent=dyn_hairTrndrilC_02_r, loc=<15.4223 0 0>, rot=<0.0221 0.0275 -0.0006 -0.9994>}
 dyn_hairTrndrilC_04_r {parent=dyn_hairTrndrilC_03_r, loc=<9.3873 0 0>, rot=<-0.0229 -0.0272 -0.0239 -0.9991>}
 dyn_hairTrndrilC_05_r {parent=dyn_hairTrndrilC_04_r, loc=<9.5123 0 0>, rot=<0 0 0 -1>}
 dyn_hairTrndrilD_01_Base_r {parent=head, loc=<11.0959 -4.5696 4.7303>, rot=<0.4956 -0.8434 0.1799 0.1037>}
 dyn_hairTrndrilD_01_r {parent=dyn_hairTrndrilD_01_Base_r, loc=<0 0 0>, rot=<0 0 0 1>}
 dyn_hairTrndrilD_02_r {parent=dyn_hairTrndrilD_01_r, loc=<12.0894 0 0>, rot=<0.0001 0.0002 0.0109 -0.9999>}
 dyn_hairTrndrilD_03_r {parent=dyn_hairTrndrilD_02_r, loc=<9.133 0 0>, rot=<-0.0052 -0.0094 -0.0086 -0.9999>}
 dyn_hairTrndrilD_04_r {parent=dyn_hairTrndrilD_03_r, loc=<8.4212 0 0>, rot=<0.0454 0.0813 0.0032 -0.9956>}
 dyn_hairTrndrilD_05_r {parent=dyn_hairTrndrilD_04_r, loc=<6.4444 0 0>, rot=<-0.0461 -0.082 -0.0059 -0.9955>}
 dyn_hairTrndrilD_06_r {parent=dyn_hairTrndrilD_05_r, loc=<8.998 0 0>, rot=<0 0 0 1>}
 dyn_hairTrndrilE_01_Base_r {parent=head, loc=<11.6783 -5.3397 3.5286>, rot=<0.5685 -0.821 0.044 0.0299>}
 dyn_hairTrndrilE_01_r {parent=dyn_hairTrndrilE_01_Base_r, loc=<0 0 0>, rot=<0 0 0 -1>}
 dyn_hairTrndrilE_02_r {parent=dyn_hairTrndrilE_01_r, loc=<10.1626 0 0>, rot=<0.0252 0.0558 -0.056 -0.9966>}
 dyn_hairTrndrilE_03_r {parent=dyn_hairTrndrilE_02_r, loc=<8.5172 0 0>, rot=<-0.0239 -0.0526 0.0539 -0.9969>}
 dyn_hairTrndrilE_04_r {parent=dyn_hairTrndrilE_03_r, loc=<9.2479 0 0>, rot=<0.0413 0.0994 -0.0217 -0.9939>}
 dyn_hairTrndrilE_05_r {parent=dyn_hairTrndrilE_04_r, loc=<8.7779 0 0>, rot=<-0.0455 -0.099 -0.0151 -0.9939>}
 dyn_hairTrndrilE_06_r {parent=dyn_hairTrndrilE_05_r, loc=<9.8583 0 0>, rot=<0 0 0 -1>}
]
