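# Physics asset for the face accessory of the female medium Jellyfish player cosmetic, providing bodies for dynamic hair bones.
# Five chains (left and right temple; left, center and right back) each run from a small base body through three capsule-based segments with limited swing and twist.
# Chain bases attach to the head with free angular motion, as do head, neck and upper-spine links, whose large capsules give the hair body collision shapes.
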PhysicsAsset /BRCosmetics/Characters/Player/Female/Medium/Bodies/F_MED_Jellyfish/Parts/F_MED_Jellyfish_FaceAcc_Physics
bodies {
 dyn_hairBack_01_Base_l [{shapeType=Box, extentX=0.5, extentY=2.25, extentZ=0.5, rotation=<rot 90 0 0>}]
 dyn_hairBack_01_Base_r [{shapeType=Box, extentX=0.5, extentY=2.25, extentZ=0.5, rotation=<rot 90 0 0>}]
 dyn_hairBack_01_c [{shapeType=Capsule, length=5, radius=3, center=<4.1338 0 0>, rotation=<rot -0.0002 0 0>}]
 dyn_hairBack_01_c_Base [{shapeType=Box, extentX=0.5, extentY=2.25, extentZ=0.5, rotation=<rot 90 0 0>}]
 dyn_hairBack_01_l [{shapeType=Capsule, length=5, radius=3, center=<4.1338 0 0>, rotation=<rot 0.0008 -179.9999 -180>}]
 dyn_hairBack_01_r [{shapeType=Capsule, length=5, radius=3, center=<4.1338 0 0>, rotation=<rot -0.0002 0 0>}]
 dyn_hairBack_02_c [{shapeType=Capsule, length=5, radius=3, center=<4.1338 0 0>, rotation=<rot -0.0002 0.0001 0>}]
 dyn_hairBack_02_l [{shapeType=Capsule, length=5, radius=3, center=<4.1338 0 0>, rotation=<rot 0.0008 -179.9999 180>}]
 dyn_hairBack_02_r [{shapeType=Capsule, length=5, radius=3, center=<4.1338 0 0>, rotation=<rot 0.0027 0 0>}]
 dyn_hairBack_03_c [{shapeType=Capsule, length=5, radius=3, center=<4.1338 0 0>, rotation=<rot -0.0002 0.0001 0>}]
 dyn_hairBack_03_l [{shapeType=Capsule, length=5, radius=3, center=<4.1338 0 0>, rotation=<rot -0.0061 -0.0001 0>}]
 dyn_hairBack_03_r [{shapeType=Capsule, length=5, radius=3, center=<4.1338 0 0>, rotation=<rot -0.0002 0 0>}]
 dyn_hairTemple_01_Base_l [{shapeType=Capsule, length=0.505, radius=0.505, rotation=<rot 0 -90 0>}]
 dyn_hairTemple_01_Base_r [{shapeType=Capsule, length=0.505, radius=0.505, rotation=<rot 0 -90 0>}]
 dyn_hairTemple_01_l [{shapeType=Capsule, length=3.8325, radius=2.0303, center=<2.7348 0 0>, rotation=<rot 90 0 0>}]
 dyn_hairTemple_01_r [{shapeType=Capsule, length=3.8325, radius=2.0303, center=<2.7348 0 0>, rotation=<rot 90 0 0>}]
 dyn_hairTemple_02_l [
  {shapeType=Capsule, length=3.886, radius=1.7872, center=<3.4644 -0.7614 0.0001>, rotation=<rot 90 0 0>}
  {shapeType=Capsule, length=2.625, radius=3.625, center=<6.4405 0.6127 0.2243>, rotation=<rot 7.1857 166.2935 -173.489>}
 ]
 dyn_hairTemple_02_r [
  {shapeType=Capsule, length=2.886, radius=2.7872, center=<3.4665 0.613 2.4171>, rotation=<rot 90 180 -169.9999>}
  {shapeType=Capsule, length=2.875, radius=3.625, center=<4.9413 0.2327 0.5534>, rotation=<rot -11.523 9.5088 -39.978>}
 ]
 dyn_hairTemple_03_l [{shapeType=Capsule, length=5.125, radius=2.505, center=<3.0761 0.0212 -0.0001>, rotation=<rot 90 162.646 173.827>}]
 dyn_hairTemple_03_r [{shapeType=Capsule, length=3.375, radius=2.755, center=<2.5739 0.1359 -0.0001>, rotation=<rot 90 180 -178.8191>}]
 head [{shapeType=Capsule, length=8.7204, radius=5.959, center=<-2.2019 -0.3657 -0.1455>, rotation=<rot 0 92.626 -86.9417>}]
 neck_01 [{shapeType=Capsule, length=43.005, radius=4.255, center=<1.9498 1.1069 0>, rotation=<rot 0 -90 0>}]
 spine_04 [{shapeType=Capsule, length=37.755, radius=9.005, center=<9.5129 3.4049 0>, rotation=<rot 0 -90 0>}]
 spine_05 [{shapeType=Capsule, length=23.255, radius=5.505, center=<3.7663 -1.4896 0>, rotation=<rot 0 -90 0>}]
}
constraints [
 {
  bone1Name=dyn_hairTemple_01_Base_r
  bone2Name=head
  swing1LimitDegrees=45
  swing1Motion=Free
  swing2LimitDegrees=45
  swing2Motion=Free
  twistLimitDegrees=45
  twistMotion=Free
 }
 {bone1Name=dyn_hairTemple_01_r, bone2Name=dyn_hairTemple_01_Base_r, swing1LimitDegrees=15, swing2LimitDegrees=30, twistLimitDegrees=10}
 {bone1Name=dyn_hairTemple_02_r, bone2Name=dyn_hairTemple_01_r, swing1LimitDegrees=30, swing2LimitDegrees=15, twistLimitDegrees=10}
 {bone1Name=dyn_hairTemple_03_r, bone2Name=dyn_hairTemple_02_r, swing1LimitDegrees=30, swing2LimitDegrees=15, twistLimitDegrees=10}
 {
  bone1Name=head
  bone2Name=neck_01
  swing1LimitDegrees=45
  swing1Motion=Free
  swing2LimitDegrees=45
  swing2Motion=Free
  twistLimitDegrees=45
  twistMotion=Free
 }
 {
  bone1Name=dyn_hairTemple_01_Base_l
  bone2Name=head
  swing1LimitDegrees=45
  swing1Motion=Free
  swing2LimitDegrees=45
  swing2Motion=Free
  twistLimitDegrees=45
  twistMotion=Free
 }
 {bone1Name=dyn_hairTemple_01_l, bone2Name=dyn_hairTemple_01_Base_l, swing1LimitDegrees=15, swing2LimitDegrees=30, twistLimitDegrees=10}
 {bone1Name=dyn_hairTemple_02_l, bone2Name=dyn_hairTemple_01_l, swing1LimitDegrees=30, swing2LimitDegrees=15, twistLimitDegrees=10}
 {bone1Name=dyn_hairTemple_03_l, bone2Name=dyn_hairTemple_02_l, swing1LimitDegrees=30, swing2LimitDegrees=15, twistLimitDegrees=10}
 {
  bone1Name=dyn_hairBack_01_Base_l
  bone2Name=head
  swing1LimitDegrees=45
  swing1Motion=Free
  swing2LimitDegrees=45
  swing2Motion=Free
  twistLimitDegrees=45
  twistMotion=Free
 }
 {bone1Name=dyn_hairBack_01_l, bone2Name=dyn_hairBack_01_Base_l, swing1LimitDegrees=35, swing2LimitDegrees=35, twistLimitDegrees=8}
 {bone1Name=dyn_hairBack_02_l, bone2Name=dyn_hairBack_01_l, swing1LimitDegrees=30, swing2LimitDegrees=30, twistLimitDegrees=10}
 {bone1Name=dyn_hairBack_03_l, bone2Name=dyn_hairBack_02_l, swing1LimitDegrees=30, swing2LimitDegrees=30, twistLimitDegrees=8}
 {
  bone1Name=dyn_hairBack_01_c_Base
  bone2Name=head
  swing1LimitDegrees=45
  swing1Motion=Free
  swing2LimitDegrees=45
  swing2Motion=Free
  twistLimitDegrees=45
  twistMotion=Free
 }
 {bone1Name=dyn_hairBack_01_c, bone2Name=dyn_hairBack_01_c_Base, swing1LimitDegrees=35, swing2LimitDegrees=35, twistLimitDegrees=8}
 {bone1Name=dyn_hairBack_02_c, bone2Name=dyn_hairBack_01_c, swing1LimitDegrees=30, swing2LimitDegrees=30, twistLimitDegrees=10}
 {bone1Name=dyn_hairBack_03_c, bone2Name=dyn_hairBack_02_c, swing1LimitDegrees=30, swing2LimitDegrees=30, twistLimitDegrees=8}
 {
  bone1Name=dyn_hairBack_01_Base_r
  bone2Name=head
  swing1LimitDegrees=45
  swing1Motion=Free
  swing2LimitDegrees=45
  swing2Motion=Free
  twistLimitDegrees=45
  twistMotion=Free
 }
 {bone1Name=dyn_hairBack_01_r, bone2Name=dyn_hairBack_01_Base_r, swing1LimitDegrees=35, swing2LimitDegrees=35, twistLimitDegrees=8}
 {bone1Name=dyn_hairBack_02_r, bone2Name=dyn_hairBack_01_r, swing1LimitDegrees=30, swing2LimitDegrees=30, twistLimitDegrees=10}
 {bone1Name=dyn_hairBack_03_r, bone2Name=dyn_hairBack_02_r, swing1LimitDegrees=30, swing2LimitDegrees=30, twistLimitDegrees=8}
 {
  bone1Name=neck_01
  bone2Name=spine_05
  swing1LimitDegrees=45
  swing1Motion=Free
  swing2LimitDegrees=45
  swing2Motion=Free
  twistLimitDegrees=45
  twistMotion=Free
 }
 {
  bone1Name=spine_05
  bone2Name=spine_04
  swing1LimitDegrees=45
  swing1Motion=Free
  swing2LimitDegrees=45
  swing2Motion=Free
  twistLimitDegrees=45
  twistMotion=Free
 }
]
