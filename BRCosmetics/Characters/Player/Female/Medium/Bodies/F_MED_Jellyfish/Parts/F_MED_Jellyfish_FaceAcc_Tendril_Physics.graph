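# Physics asset for the tendrils of the female medium Jellyfish cosmetic's face accessory, giving them bodies and joints for physical motion.
# Eight capsule chains (B to E, left and right) each run from a tiny base body on the head through four segments.
# Segment joints have narrow swing and twist limits, tighter at the left D and E roots, while base-to-head joints are angularly free.
# Head, clavicle and upper-spine capsules provide body shapes, linked by free-rotation constraints.
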PhysicsAsset /BRCosmetics/Characters/Player/Female/Medium/Bodies/F_MED_Jellyfish/Parts/F_MED_Jellyfish_FaceAcc_Tendril_Physics
bodies {
 clavicle_l [{shapeType=Capsule, length=20.88, radius=4.38, center=<7.95 -0.9621 0>, rotation=<rot 90 0 174.0397>}]
 clavicle_r [{shapeType=Capsule, length=20.755, radius=4.505, center=<-8.5871 1.0422 0>, rotation=<rot 90 0 -5.9603>}]
 dyn_hairTrndrilB_01_Base_l [{shapeType=Capsule, length=0.505, radius=0.505, rotation=<rot 0 -90 0>}]
 dyn_hairTrndrilB_01_Base_r [{shapeType=Capsule, length=0.505, radius=0.505, rotation=<rot 0 -90 0>}]
 dyn_hairTrndrilB_01_l [{shapeType=Capsule, length=12.875, radius=1.88, center=<5.0617 0 0>, rotation=<rot 90 0 0>}]
 dyn_hairTrndrilB_01_r [{shapeType=Capsule, length=12.875, radius=1.88, center=<5.0617 0 0>, rotation=<rot 90 0 0>}]
 dyn_hairTrndrilB_02_l [{shapeType=Capsule, length=12.875, radius=1.88, center=<5.0617 0 0>, rotation=<rot 90 0 0>}]
 dyn_hairTrndrilB_02_r [{shapeType=Capsule, length=12.875, radius=1.88, center=<5.0617 0 0>, rotation=<rot 90 0 0>}]
 dyn_hairTrndrilB_03_l [{shapeType=Capsule, length=12.875, radius=1.88, center=<5.0617 0 0>, rotation=<rot 90 0 0>}]
 dyn_hairTrndrilB_03_r [{shapeType=Capsule, length=12.875, radius=1.88, center=<5.0617 0 0>, rotation=<rot 90 0 0>}]
 dyn_hairTrndrilB_04_l [{shapeType=Capsule, length=12.875, radius=1.88, center=<5.0617 0 -1.3313>, rotation=<rot 76.3467 -0.0001 0>}]
 dyn_hairTrndrilB_04_r [{shapeType=Capsule, length=12.875, radius=1.88, center=<5.0617 0 0>, rotation=<rot 90 0 0>}]
 dyn_hairTrndrilC_01_Base_l [{shapeType=Capsule, length=0.505, radius=0.505, rotation=<rot 0 -90 0>}]
 dyn_hairTrndrilC_01_Base_r [{shapeType=Capsule, length=0.505, radius=0.505, rotation=<rot 0 -90 0>}]
 dyn_hairTrndrilC_01_l [{shapeType=Capsule, length=12.875, radius=1.88, center=<5.0617 0 0>, rotation=<rot 90 0 0>}]
 dyn_hairTrndrilC_01_r [{shapeType=Capsule, length=12.875, radius=1.88, center=<5.0617 0 0>, rotation=<rot 90 0 0>}]
 dyn_hairTrndrilC_02_l [{shapeType=Capsule, length=12.875, radius=1.88, center=<5.0617 0 0>, rotation=<rot 90 0 0>}]
 dyn_hairTrndrilC_02_r [{shapeType=Capsule, length=12.875, radius=1.88, center=<5.0617 0 0>, rotation=<rot 90 0 0>}]
 dyn_hairTrndrilC_03_l [{shapeType=Capsule, length=12.875, radius=1.88, center=<5.0617 0 0>, rotation=<rot 90 0 0>}]
 dyn_hairTrndrilC_03_r [{shapeType=Capsule, length=12.875, radius=1.88, center=<5.0617 0 0>, rotation=<rot 90 0 0>}]
 dyn_hairTrndrilC_04_l [{shapeType=Capsule, length=12.875, radius=1.88, center=<5.0617 0 0>, rotation=<rot 90 0 0>}]
 dyn_hairTrndrilC_04_r [{shapeType=Capsule, length=12.875, radius=1.88, center=<5.0617 0 0>, rotation=<rot 90 0 0>}]
 dyn_hairTrndrilD_01_Base_l [{shapeType=Capsule, length=0.505, radius=0.505, rotation=<rot 0 -90 0>}]
 dyn_hairTrndrilD_01_Base_r [{shapeType=Capsule, length=0.505, radius=0.505, rotation=<rot 0 -90 0>}]
 dyn_hairTrndrilD_01_l [{shapeType=Capsule, length=12.875, radius=1.88, center=<5.0617 0 0>, rotation=<rot 90 0 0>}]
 dyn_hairTrndrilD_01_r [{shapeType=Capsule, length=12.875, radius=1.88, center=<5.0617 0 0>, rotation=<rot 90 0 0>}]
 dyn_hairTrndrilD_02_l [{shapeType=Capsule, length=12.875, radius=1.88, center=<5.0617 0 0>, rotation=<rot 90 0 0>}]
 dyn_hairTrndrilD_02_r [{shapeType=Capsule, length=12.875, radius=1.88, center=<5.0617 0 0>, rotation=<rot 90 0 0>}]
 dyn_hairTrndrilD_03_l [{shapeType=Capsule, length=12.875, radius=1.88, center=<5.0617 0 0>, rotation=<rot 90 0 0>}]
 dyn_hairTrndrilD_03_r [{shapeType=Capsule, length=12.875, radius=1.88, center=<5.0617 0 0>, rotation=<rot 90 0 0>}]
 dyn_hairTrndrilD_04_l [{shapeType=Capsule, length=12.875, radius=1.88, center=<5.0617 0 0>, rotation=<rot 90 0 0>}]
 dyn_hairTrndrilD_04_r [{shapeType=Capsule, length=12.875, radius=1.88, center=<5.0617 0 0>, rotation=<rot 90 0 0>}]
 dyn_hairTrndrilE_01_Base_l [{shapeType=Capsule, length=0.505, radius=0.505, rotation=<rot 0 -90 0>}]
 dyn_hairTrndrilE_01_Base_r [{shapeType=Capsule, length=0.505, radius=0.505, rotation=<rot 0 -90 0>}]
 dyn_hairTrndrilE_01_l [{shapeType=Capsule, length=12.875, radius=1.88, center=<5.0617 0 0>, rotation=<rot 90 0 0>}]
 dyn_hairTrndrilE_01_r [{shapeType=Capsule, length=12.875, radius=1.88, center=<5.0617 0 0>, rotation=<rot 90 0 0>}]
 dyn_hairTrndrilE_02_l [{shapeType=Capsule, length=12.875, radius=1.88, center=<5.0617 0 0>, rotation=<rot 90 0 0>}]
 dyn_hairTrndrilE_02_r [{shapeType=Capsule, length=12.875, radius=1.88, center=<5.0617 0 0>, rotation=<rot 90 0 0>}]
 dyn_hairTrndrilE_03_l [{shapeType=Capsule, length=12.875, radius=1.88, center=<5.0617 0 0>, rotation=<rot 90 0 0>}]
 dyn_hairTrndrilE_03_r [{shapeType=Capsule, length=12.875, radius=1.88, center=<5.0617 0 0>, rotation=<rot 90 0 0>}]
 dyn_hairTrndrilE_04_l [{shapeType=Capsule, length=12.875, radius=1.88, center=<5.0617 0 0>, rotation=<rot 90 0 0>}]
 dyn_hairTrndrilE_04_r [{shapeType=Capsule, length=12.875, radius=1.88, center=<5.0617 0 0>, rotation=<rot 90 0 0>}]
 head [{shapeType=Capsule, length=9.2385, radius=8.9448, center=<0.6548 -1.2093 0>, rotation=<rot 90 0 0>}]
 spine_04 [{shapeType=Capsule, length=17.63, radius=14.755, center=<-5.1663 4.3529 -0.0003>, rotation=<rot 90 0 0>}]
 spine_05 [{shapeType=Capsule, length=12.63, radius=14.755, center=<-6.2645 3.2734 0>, rotation=<rot 90 0 0>}]
}
constraints [
 {bone1Name=dyn_hairTrndrilE_01_l, bone2Name=dyn_hairTrndrilE_01_Base_l, swing1LimitDegrees=10, swing2LimitDegrees=25, twistLimitDegrees=15}
 {bone1Name=dyn_hairTrndrilE_02_l, bone2Name=dyn_hairTrndrilE_01_l, swing1LimitDegrees=25, swing2LimitDegrees=25, twistLimitDegrees=15}
 {bone1Name=dyn_hairTrndrilD_01_l, bone2Name=dyn_hairTrndrilD_01_Base_l, swing1LimitDegrees=10, swing2LimitDegrees=25, twistLimitDegrees=15}
 {bone1Name=dyn_hairTrndrilD_02_l, bone2Name=dyn_hairTrndrilD_01_l, swing1LimitDegrees=25, swing2LimitDegrees=25, twistLimitDegrees=15}
 {bone1Name=dyn_hairTrndrilC_01_l, bone2Name=dyn_hairTrndrilC_01_Base_l, swing1LimitDegrees=25, swing2LimitDegrees=25, twistLimitDegrees=15}
 {bone1Name=dyn_hairTrndrilC_02_l, bone2Name=dyn_hairTrndrilC_01_l, swing1LimitDegrees=25, swing2LimitDegrees=25, twistLimitDegrees=15}
 {bone1Name=dyn_hairTrndrilB_01_l, bone2Name=dyn_hairTrndrilB_01_Base_l, swing1LimitDegrees=25, swing2LimitDegrees=25, twistLimitDegrees=15}
 {bone1Name=dyn_hairTrndrilB_02_l, bone2Name=dyn_hairTrndrilB_01_l, swing1LimitDegrees=25, swing2LimitDegrees=25, twistLimitDegrees=15}
 {bone1Name=dyn_hairTrndrilB_01_r, bone2Name=dyn_hairTrndrilB_01_Base_r, swing1LimitDegrees=25, swing2LimitDegrees=25, twistLimitDegrees=15}
 {bone1Name=dyn_hairTrndrilB_02_r, bone2Name=dyn_hairTrndrilB_01_r, swing1LimitDegrees=25, swing2LimitDegrees=25, twistLimitDegrees=15}
 {bone1Name=dyn_hairTrndrilC_01_r, bone2Name=dyn_hairTrndrilC_01_Base_r, swing1LimitDegrees=25, swing2LimitDegrees=25, twistLimitDegrees=15}
 {bone1Name=dyn_hairTrndrilC_02_r, bone2Name=dyn_hairTrndrilC_01_r, swing1LimitDegrees=25, swing2LimitDegrees=25, twistLimitDegrees=15}
 {bone1Name=dyn_hairTrndrilD_01_r, bone2Name=dyn_hairTrndrilD_01_Base_r, swing1LimitDegrees=25, swing2LimitDegrees=25, twistLimitDegrees=15}
 {bone1Name=dyn_hairTrndrilD_02_r, bone2Name=dyn_hairTrndrilD_01_r, swing1LimitDegrees=25, swing2LimitDegrees=25, twistLimitDegrees=15}
 {bone1Name=dyn_hairTrndrilE_01_r, bone2Name=dyn_hairTrndrilE_01_Base_r, swing1LimitDegrees=25, swing2LimitDegrees=25, twistLimitDegrees=15}
 {bone1Name=dyn_hairTrndrilE_02_r, bone2Name=dyn_hairTrndrilE_01_r, swing1LimitDegrees=25, swing2LimitDegrees=25, twistLimitDegrees=15}
 {
  bone1Name=dyn_hairTrndrilE_01_Base_l
  bone2Name=head
  swing1LimitDegrees=45
  swing1Motion=Free
  swing2LimitDegrees=45
  swing2Motion=Free
  twistLimitDegrees=45
  twistMotion=Free
 }
 {
  bone1Name=dyn_hairTrndrilD_01_Base_l
  bone2Name=head
  swing1LimitDegrees=45
  swing1Motion=Free
  swing2LimitDegrees=45
  swing2Motion=Free
  twistLimitDegrees=45
  twistMotion=Free
 }
 {
  bone1Name=dyn_hairTrndrilC_01_Base_l
  bone2Name=head
  swing1LimitDegrees=45
  swing1Motion=Free
  swing2LimitDegrees=45
  swing2Motion=Free
  twistLimitDegrees=45
  twistMotion=Free
 }
 {
  bone1Name=dyn_hairTrndrilB_01_Base_l
  bone2Name=head
  swing1LimitDegrees=45
  swing1Motion=Free
  swing2LimitDegrees=45
  swing2Motion=Free
  twistLimitDegrees=45
  twistMotion=Free
 }
 {
  bone1Name=dyn_hairTrndrilB_01_Base_r
  bone2Name=head
  swing1LimitDegrees=45
  swing1Motion=Free
  swing2LimitDegrees=45
  swing2Motion=Free
  twistLimitDegrees=45
  twistMotion=Free
 }
 {
  bone1Name=dyn_hairTrndrilC_01_Base_r
  bone2Name=head
  swing1LimitDegrees=45
  swing1Motion=Free
  swing2LimitDegrees=45
  swing2Motion=Free
  twistLimitDegrees=45
  twistMotion=Free
 }
 {
  bone1Name=dyn_hairTrndrilD_01_Base_r
  bone2Name=head
  swing1LimitDegrees=45
  swing1Motion=Free
  swing2LimitDegrees=45
  swing2Motion=Free
  twistLimitDegrees=45
  twistMotion=Free
 }
 {
  bone1Name=dyn_hairTrndrilE_01_Base_r
  bone2Name=head
  swing1LimitDegrees=45
  swing1Motion=Free
  swing2LimitDegrees=45
  swing2Motion=Free
  twistLimitDegrees=45
  twistMotion=Free
 }
 {bone1Name=dyn_hairTrndrilE_03_r, bone2Name=dyn_hairTrndrilE_02_r, swing1LimitDegrees=25, swing2LimitDegrees=25, twistLimitDegrees=15}
 {bone1Name=dyn_hairTrndrilE_04_r, bone2Name=dyn_hairTrndrilE_03_r, swing1LimitDegrees=25, swing2LimitDegrees=25, twistLimitDegrees=15}
 {bone1Name=dyn_hairTrndrilD_03_r, bone2Name=dyn_hairTrndrilD_02_r, swing1LimitDegrees=25, swing2LimitDegrees=25, twistLimitDegrees=15}
 {bone1Name=dyn_hairTrndrilD_04_r, bone2Name=dyn_hairTrndrilD_03_r, swing1LimitDegrees=25, swing2LimitDegrees=25, twistLimitDegrees=15}
 {bone1Name=dyn_hairTrndrilC_03_r, bone2Name=dyn_hairTrndrilC_02_r, swing1LimitDegrees=25, swing2LimitDegrees=25, twistLimitDegrees=15}
 {bone1Name=dyn_hairTrndrilC_04_r, bone2Name=dyn_hairTrndrilC_03_r, swing1LimitDegrees=25, swing2LimitDegrees=25, twistLimitDegrees=15}
 {bone1Name=dyn_hairTrndrilB_03_r, bone2Name=dyn_hairTrndrilB_02_r, swing1LimitDegrees=25, swing2LimitDegrees=25, twistLimitDegrees=15}
 {bone1Name=dyn_hairTrndrilB_04_r, bone2Name=dyn_hairTrndrilB_03_r, swing1LimitDegrees=25, swing2LimitDegrees=25, twistLimitDegrees=15}
 {bone1Name=dyn_hairTrndrilB_03_l, bone2Name=dyn_hairTrndrilB_02_l, swing1LimitDegrees=25, swing2LimitDegrees=25, twistLimitDegrees=15}
 {bone1Name=dyn_hairTrndrilB_04_l, bone2Name=dyn_hairTrndrilB_03_l, swing1LimitDegrees=25, swing2LimitDegrees=25, twistLimitDegrees=15}
 {bone1Name=dyn_hairTrndrilC_03_l, bone2Name=dyn_hairTrndrilC_02_l, swing1LimitDegrees=25, swing2LimitDegrees=25, twistLimitDegrees=15}
 {bone1Name=dyn_hairTrndrilC_04_l, bone2Name=dyn_hairTrndrilC_03_l, swing1LimitDegrees=25, swing2LimitDegrees=25, twistLimitDegrees=15}
 {bone1Name=dyn_hairTrndrilD_03_l, bone2Name=dyn_hairTrndrilD_02_l, swing1LimitDegrees=25, swing2LimitDegrees=25, twistLimitDegrees=15}
 {bone1Name=dyn_hairTrndrilD_04_l, bone2Name=dyn_hairTrndrilD_03_l, swing1LimitDegrees=25, swing2LimitDegrees=25, twistLimitDegrees=15}
 {bone1Name=dyn_hairTrndrilE_03_l, bone2Name=dyn_hairTrndrilE_02_l, swing1LimitDegrees=25, swing2LimitDegrees=25, twistLimitDegrees=15}
 {bone1Name=dyn_hairTrndrilE_04_l, bone2Name=dyn_hairTrndrilE_03_l, swing1LimitDegrees=25, swing2LimitDegrees=25, twistLimitDegrees=15}
 {
  bone1Name=clavicle_r
  bone2Name=spine_05
  swing1LimitDegrees=45
  swing1Motion=Free
  swing2LimitDegrees=45
  swing2Motion=Free
  twistLimitDegrees=45
  twistMotion=Free
 }
 {
  bone1Name=head
  bone2Name=spine_05
  swing1LimitDegrees=45
  swing1Motion=Free
  swing2LimitDegrees=45
  swing2Motion=Free
  twistLimitDegrees=45
  twistMotion=Free
 }
 {
  bone1Name=spine_05
  bone2Name=spine_04
  swing1LimitDegrees=45
  swing1Motion=Free
  swing2LimitDegrees=45
  swing2Motion=Free
  twistLimitDegrees=45
  twistMotion=Free
 }
 {
  bone1Name=clavicle_l
  bone2Name=spine_05
  swing1LimitDegrees=45
  swing1Motion=Free
  swing2LimitDegrees=45
  swing2Motion=Free
  twistLimitDegrees=45
  twistMotion=Free
 }
]
