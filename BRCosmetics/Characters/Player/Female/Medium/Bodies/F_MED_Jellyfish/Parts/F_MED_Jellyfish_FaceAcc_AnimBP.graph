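# Animation Blueprint for the Jellyfish female character's face accessory, posing and simulating its hair tendrils.
# With rigid bodies enabled, parent and body-mesh limb poses feed two chained RigidBody simulations, then a tendril Control Rig.
# Otherwise, fixed additive bone offsets pose the tendrils.
# The event graph derives head and arm swing values from bone deltas, and resets cloth and dynamics after delays when skydiving ends or the parachute opens.
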AnimBlueprint /BRCosmetics/Characters/Player/Female/Medium/Bodies/F_MED_Jellyfish/Parts/F_MED_Jellyfish_FaceAcc_AnimBP
target_skeleton=&/F_MED_Jellyfish_FaceAcc_Skeleton
graphs [
 AnimGraph {nodes [
   Output Pose(Result=Blend Poses by bool(BlendPose_0=Use cached pose 'RBANCache'(), BlendPose_1=Use cached pose 'nonRBANCache'(), bActiveValue=Get bRigidBodyEnabled()) {blendType=Linear}) {layerGroup=None}
   n1 = Layered blend per bone(
    BasePose=Copy Pose From Mesh() {bCopyCurves=true, bUseAttachedParent=true}
    BlendPoses_0=Copy Pose From Mesh(SourceMeshComponent=Get bodyMesh() -> (bodyMesh:Skeletal Mesh Component Object Reference))
   ) {layerSetup [{branchFilters [
       {boneName=upperarm_l}
       {boneName=upperarm_r}
       {boneName=lowerarm_l}
       {boneName=lowerarm_r}
       {boneName=thigh_l}
       {boneName=thigh_r}
       {boneName=calf_l}
       {boneName=calf_r}
      ]}]}
   n2 = RigidBody(ComponentPose=Local To Component(LocalPose=@n1)) {
    bClampLinearTranslationLimitToRefPose=true
    bOverrideWorldGravity=true
    overridePhysicsAsset=&/F_MED_Jellyfish_FaceAcc_Physics
    simulationSpace=BaseBoneSpace
    alphaBoolBlend {blendOption=Linear}
    baseBoneRef {boneName=spine_02}
    componentAppliedLinearAccClamp=<4000 4000 150>
    componentLinearAccScale=<0.5 0.5 0.5>
    componentLinearVelScale=<0.35 0.35 0.35>
    overrideWorldGravity {z=-980}
   }
   n3 = RigidBody(ComponentPose=@n2) {
    bClampLinearTranslationLimitToRefPose=true
    bOverrideWorldGravity=true
    overridePhysicsAsset=&/F_MED_Jellyfish_FaceAcc_Tendril_Physics
    simulationSpace=BaseBoneSpace
    alphaBoolBlend {blendOption=Linear}
    baseBoneRef {boneName=head}
    componentAppliedLinearAccClamp=<4000 4000 300>
    componentLinearAccScale=<0.5 0.5 0.3>
    componentLinearVelScale=<0.35 0.35 0.25>
    overrideWorldGravity {z=-250}
   }
   n4 = Control Rig(Source=Component To Local(ComponentPose=@n3)) {
    alphaBoolBlend {blendOption=Linear}
    controlRigAssetReference {blueprintRigClass=&/F_MED_JellyfishTendril_ControlRig.F_MED_JellyfishTendril_ControlRig_C}
   }
   Save cached pose 'RBANCache'(Pose=@n4) {cacheName=RBANCache}
   n5 = "Transform (Modify) Bone - Bone: dyn_hairTrndrilB_01_Base_l"(
    ComponentPose=Local To Component(LocalPose=Copy Pose From Mesh() {bUseAttachedParent=true})
    Alpha=1
    Translation="0,0,0"
    Rotation="0.000,7.498,-0.000"
    Scale="1,1,1"
   ) {
    rotationMode=BMM_Additive
    rotationSpace=BCS_BoneSpace
    alphaBoolBlend {blendOption=Linear}
    boneToModify {boneName=dyn_hairTrndrilB_01_Base_l}
    rotation=<rot 0 7.498 0>
   }
   n6 = "Transform (Modify) Bone - Bone: dyn_hairTrndrilC_01_Base_l"(ComponentPose=@n5, Alpha=1, Translation="0,0,0", Rotation="0.000,26.458,-0.000", Scale="1,1,1") {
    rotationMode=BMM_Additive
    rotationSpace=BCS_BoneSpace
    alphaBoolBlend {blendOption=Linear}
    boneToModify {boneName=dyn_hairTrndrilC_01_Base_l}
    rotation {yaw=26.458}
   }
   n7 = "Transform (Modify) Bone - Bone: dyn_hairTrndrilD_01_Base_l"(ComponentPose=@n6, Alpha=1, Translation="0,0,0", Rotation="0.000,13.180,-0.000", Scale="1,1,1") {
    rotationMode=BMM_Additive
    rotationSpace=BCS_BoneSpace
    alphaBoolBlend {blendOption=Linear}
    boneToModify {boneName=dyn_hairTrndrilD_01_Base_l}
    rotation=<rot 0 13.18 0>
   }
   n8 = "Transform (Modify) Bone - Bone: dyn_hairTrndrilE_01_Base_l"(ComponentPose=@n7, Alpha=1, Translation="0,0,0", Rotation="0.000,7.661,-0.000", Scale="1,1,1") {
    rotationMode=BMM_Additive
    rotationSpace=BCS_BoneSpace
    alphaBoolBlend {blendOption=Linear}
    boneToModify {boneName=dyn_hairTrndrilE_01_Base_l}
    rotation=<rot 0 7.661 0>
   }
   n9 = "Transform (Modify) Bone - Bone: dyn_hairTrndrilB_01_Base_r"(ComponentPose=@n8, Alpha=1, Translation="0,0,0", Rotation="-0.019,5.810,-0.030", Scale="1,1,1") {
    rotationMode=BMM_Additive
    rotationSpace=BCS_BoneSpace
    alphaBoolBlend {blendOption=Linear}
    boneToModify {boneName=dyn_hairTrndrilB_01_Base_r}
    rotation=<rot -0.019 5.81 -0.03>
   }
   n10 = "Transform (Modify) Bone - Bone: dyn_hairTrndrilC_01_Base_r"(ComponentPose=@n9, Alpha=1, Translation="0,0,0", Rotation="-0.051,22.822,-0.005", Scale="1,1,1") {
    rotationMode=BMM_Additive
    rotationSpace=BCS_BoneSpace
    alphaBoolBlend {blendOption=Linear}
    boneToModify {boneName=dyn_hairTrndrilC_01_Base_r}
    rotation=<rot -0.051 22.822 -0.005>
   }
   n11 = "Transform (Modify) Bone - Bone: dyn_hairTrndrilD_01_Base_r"(ComponentPose=@n10, Alpha=1, Translation="0,0,0", Rotation="-0.000,19.548,0.000", Scale="1,1,1") {
    rotationMode=BMM_Additive
    rotationSpace=BCS_BoneSpace
    alphaBoolBlend {blendOption=Linear}
    boneToModify {boneName=dyn_hairTrndrilD_01_Base_r}
    rotation=<rot 0 19.548 0>
   }
   n12 = "Transform (Modify) Bone - Bone: dyn_hairTrndrilE_01_Base_r"(ComponentPose=@n11, Alpha=1, Translation="0,0,0", Rotation="-0.000,21.428,-0.000", Scale="1,1,1") {
    rotationMode=BMM_Additive
    rotationSpace=BCS_BoneSpace
    alphaBoolBlend {blendOption=Linear}
    boneToModify {boneName=dyn_hairTrndrilE_01_Base_r}
    rotation {yaw=21.428}
   }
   n13 = "Transform (Modify) Bone - Bone: dyn_hairTrndrilB_02_l"(ComponentPose=@n12, Alpha=1, Translation="-1.130,2.582,-0.000", Rotation="0.000,13.187,-0.000", Scale="1,1,1") {
    rotationMode=BMM_Additive
    rotationSpace=BCS_BoneSpace
    translationMode=BMM_Additive
    translationSpace=BCS_BoneSpace
    alphaBoolBlend {blendOption=Linear}
    boneToModify {boneName=dyn_hairTrndrilB_02_l}
    rotation=<rot 0 13.187 0>
    translation=<-1.13 2.582>
   }
   n14 = "Transform (Modify) Bone - Bone: dyn_hairTrndrilC_02_l"(ComponentPose=@n13, Alpha=1, Translation="0,0,0", Rotation="-0.000,-5.529,0.000", Scale="1,1,1") {
    rotationMode=BMM_Additive
    rotationSpace=BCS_BoneSpace
    alphaBoolBlend {blendOption=Linear}
    boneToModify {boneName=dyn_hairTrndrilC_02_l}
    rotation {yaw=-5.529}
   }
   n15 = "Transform (Modify) Bone - Bone: dyn_hairTrndrilD_02_l"(ComponentPose=@n14, Alpha=1, Translation="-0.744,3.174,0.000", Rotation="9.584,18.624,0.909", Scale="1,1,1") {
    rotationMode=BMM_Additive
    rotationSpace=BCS_BoneSpace
    translationMode=BMM_Additive
    translationSpace=BCS_BoneSpace
    alphaBoolBlend {blendOption=Linear}
    boneToModify {boneName=dyn_hairTrndrilD_02_l}
    rotation=<rot 9.584 18.624 0.909>
    translation=<-0.744 3.174>
   }
   n16 = "Transform (Modify) Bone - Bone: dyn_hairTrndrilE_02_l"(ComponentPose=@n15, Alpha=1, Translation="-1.879,2.415,-0.000", Rotation="26.467,37.900,0.000", Scale="1,1,1") {
    rotationMode=BMM_Additive
    rotationSpace=BCS_BoneSpace
    translationMode=BMM_Additive
    translationSpace=BCS_BoneSpace
    alphaBoolBlend {blendOption=Linear}
    boneToModify {boneName=dyn_hairTrndrilE_02_l}
    rotation=<rot 26.467 37.9 0>
    translation=<-1.879 2.415>
   }
   n17 = "Transform (Modify) Bone - Bone: dyn_hairTrndrilB_02_r"(ComponentPose=@n16, Alpha=1, Translation="-0.515,2.201,-0.000", Rotation="-0.009,2.897,0.034", Scale="1,1,1") {
    rotationMode=BMM_Additive
    rotationSpace=BCS_BoneSpace
    translationMode=BMM_Additive
    translationSpace=BCS_BoneSpace
    alphaBoolBlend {blendOption=Linear}
    boneToModify {boneName=dyn_hairTrndrilB_02_r}
    rotation=<rot -0.009 2.897 0.034>
    translation=<-0.515 2.201>
   }
   n18 = "Transform (Modify) Bone - Bone: dyn_hairTrndrilC_02_r"(ComponentPose=@n17, Alpha=1, Translation="0,0,0", Rotation="0.000,-11.795,-0.000", Scale="1,1,1") {
    rotationMode=BMM_Additive
    rotationSpace=BCS_BoneSpace
    alphaBoolBlend {blendOption=Linear}
    boneToModify {boneName=dyn_hairTrndrilC_02_r}
    rotation {yaw=-11.795}
   }
   n19 = "Transform (Modify) Bone - Bone: dyn_hairTrndrilD_02_r"(ComponentPose=@n18, Alpha=1, Translation="-0.815,2.294,-0.000", Rotation="-0.000,19.548,0.000", Scale="1,1,1") {
    rotationMode=BMM_Additive
    rotationSpace=BCS_BoneSpace
    translationMode=BMM_Additive
    translationSpace=BCS_BoneSpace
    alphaBoolBlend {blendOption=Linear}
    boneToModify {boneName=dyn_hairTrndrilD_02_r}
    rotation=<rot 0 19.548 0>
    translation=<-0.815 2.294>
   }
   n20 = "Transform (Modify) Bone - Bone: dyn_hairTrndrilE_02_r"(ComponentPose=@n19, Alpha=1, Translation="-0.871,1.734,0.000", Rotation="-0.000,26.657,-0.000", Scale="1,1,1") {
    rotationMode=BMM_Additive
    rotationSpace=BCS_BoneSpace
    translationMode=BMM_Additive
    translationSpace=BCS_BoneSpace
    alphaBoolBlend {blendOption=Linear}
    boneToModify {boneName=dyn_hairTrndrilE_02_r}
    rotation=<rot 0 26.657 0>
    translation=<-0.871 1.734>
   }
   Save cached pose 'nonRBANCache'(Pose=Component To Local(ComponentPose=@n20)) {cacheName=nonRBANCache}
  ]}
 EventGraph {nodes [
   Event BlueprintUpdateAnimation()
   > n1 = Sequence()
   n2 = GetOwnerSkeletalMeshComponent()
   n3 = BreakTransform(InTransform=GetDeltaTransformFromRefPose(self=@n2, BoneName=head, BaseName=spine_05))
   n4 = Set headSwingDelta(execute=@n1.then_1, headSwingDelta:double=MapRangeClamped(Value=BreakRotator(InRot=@n3.Rotation).Yaw, InRangeA=17, InRangeB=-44)) -> (Output_Get:double)
   n5 = BreakRotator(InRot=@n3.Rotation)
   > Set headTwist_r(headTwist_r:double=MapRangeClamped(Value=@n5, InRangeA=15, InRangeB=-37)) -> (Output_Get:double)
   > Set headTwist_l(headTwist_l:double=MapRangeClamped(Value=@n5, InRangeA=15, InRangeB=65)) -> (Output_Get:double)
   n6 = BreakTransform(InTransform=GetDeltaTransformFromRefPose(self=@n2, BoneName=upperarm_r, BaseName=spine_05))
   n7 = Set armSwingDelta_r(execute=@n1.then_0, armSwingDelta_r:double=MapRangeClamped(Value=BreakRotator(InRot=@n6.Rotation).Yaw, InRangeA=45, InRangeB=130)) -> (Output_Get:double)
   n8 = BreakTransform(InTransform=GetDeltaTransformFromRefPose(self=@n2, BoneName=upperarm_l, BaseName=spine_05))
   > n9 = Set armSwingDelta_l(armSwingDelta_l:double=MapRangeClamped(Value=BreakRotator(InRot=@n8.Rotation).Yaw, InRangeA=12, InRangeB=100)) -> (Output_Get:double)
   > Set headSwingDeltaTarget_l(headSwingDeltaTarget_l:double=float * float(A=@n9, B=@n4)) -> (Output_Get:double)
   > Set headSwingDeltaTarget_r(headSwingDeltaTarget_r:double=float * float(A=@n7, B=@n4)) -> (Output_Get:double)
   Event OnEndSkydiving()
   > Do Once(Reset=@n10.then)
   > Delay(Duration=0.25)
   > ForceClothNextUpdateTeleportAndReset(self:Skeletal Mesh Component Object Reference=Get bodyMesh())
   > ResetDynamics(InTeleportType=TeleportPhysics)
   > n10 = ResetDynamics(InTeleportType=ResetPhysics)
   Event BlueprintInitializeAnimation()
   > Set bodyMesh(bodyMesh:Skeletal Mesh Component Object Reference=GetPartSkeletalMeshComponent()) -> (Output_Get:Skeletal Mesh Component Object Reference)
   Event OnParachuteOpened()
   > Do Once(Reset=@n11.then)
   > Delay(Duration=1.4)
   > ForceClothNextUpdateTeleportAndReset(self=Get bodyMesh())
   > ResetDynamics(InTeleportType=TeleportPhysics)
   > n11 = ResetDynamics(InTeleportType=ResetPhysics)
   TryGetPawnOwner()
   TryGetPawnOwner()
   Event OnBeginSkydiving()
   Event OnParachuteClosed()
  ]}
]
variables [
 armSwingDelta_l {type=number}
 armSwingDelta_r {type=number}
 bodyMesh {type=SkeletalMeshComponent}
 headSwingDelta {type=number}
 headSwingDeltaTarget_l {type=number}
 headSwingDeltaTarget_r {type=number}
 headTwist_l {type=number}
 headTwist_r {type=number}
]
